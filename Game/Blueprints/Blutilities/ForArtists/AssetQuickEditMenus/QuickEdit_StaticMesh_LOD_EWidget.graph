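# Editor Utility Widget for artists that quick-edits static mesh LOD settings through a Details View filtered to LOD, reduction, screen size and body setup properties.
# Its setAsset function stores the object, feeds it to the Details View and updates the title and Save label.
# The magnifying-glass button syncs the Content Browser to the asset, and Save writes it only if dirty.
# Property changes just print the changed property's name.
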EditorUtilityWidgetBlueprint /Game/Blueprints/Blutilities/ForArtists/AssetQuickEditMenus/QuickEdit_StaticMesh_LOD_EWidget
graphs [
 EventGraph {nodes [
   "On Clicked (TA_Generic01_Button)"!() -> (OutputDelegate, then, widget:User Widget Object Reference, event:Pointer Event Structure) {is=title K2Node_ComponentBoundEvent}
   n1 = Get currentAsset() -> (currentAsset:Object Reference)
   > Is Valid(InputObject=@n1)
   n2 = CreateAssetData!(InAsset=@n1, bAllowBlueprintClass=false) -> (
    ReturnValue_PackageName:Name
    ReturnValue_PackagePath:Name
    ReturnValue_AssetName:Name
    ReturnValue_AssetClassPath:Top Level Asset Path Structure
   )
   > SyncBrowserToObjects(AssetPaths=Make Array("[0]":String="To String (Name)"(InName=@n2)) -> (Array:Array of Strings))
   "On Clicked (btnSave)"!() -> (OutputDelegate, then, widget:User Widget Object Reference, event:Pointer Event Structure) {is=title K2Node_ComponentBoundEvent}
   > SaveLoadedAsset(AssetToSave=Get currentAsset(), bOnlyIfIsDirty=true)
   n3 = OnPropertyChanged_Event_0!() -> (OutputDelegate, then, PropertyName:Name) {is=title K2Node_CustomEvent}
   > PrintString(InString="To String (Name)"(InName=@n3.PropertyName))
   Event Construct()
   > Assign On Property Changed!(
    execute:Exec
    self:Property View Base Object Reference=Get DetailsView_35() -> (DetailsView_35:Details View Object Reference)
    Delegate:"Delegate (by ref)"=@n3
   ) -> (then:Exec) {is=title K2Node_AssignDelegate}
   Event PreConstruct()
   Event Tick()
  ]}
 setAsset {nodes [
   n1 = setAsset!() -> (then, object:Object Reference) {is=title K2Node_FunctionEntry}
   > Set currentAsset(currentAsset:Object Reference=@n1) -> (Output_Get:Object Reference)
   > SetObject(self=Get DetailsView_35(), NewObject=@n1)
   n2 = GetDisplayName(Object=Get currentAsset())
   > "SetText (Text)"(
    self=Get CommonTextBlock_72() -> (CommonTextBlock_72:Common Text Object Reference)
    InText="To Text (String)"(InString=Append(A="Editing: ", B=@n2))
   )
   n3 = Get currentAsset()
   > setLabel!(
    execute
    self:TA Base Editor Button Widget Object Reference=Get btnSave() -> (btnSave:TA Generic 01 Button Object Reference)
    label:String=Append(A="Save: ", B=GetDisplayName(Object=@n3))
   ) -> (then)
  ]}
]
variables [currentAsset {type=Object, default=null}]
widgets {
 info {rootWidgetClass=&CanvasPanel, widgetCount=11}
 widgets [
  CanvasPanel_0 {widgetClassPath=&CanvasPanel, widget=CanvasPanel_0}
  VerticalBox_0 {
   widgetClassPath=&VerticalBox
   parent=CanvasPanel_0
   widget=VerticalBox_0
   slot {class=CanvasPanelSlot, properties {layoutData {alignment=<0 0>, anchors {maximum=<1 1>, minimum=<0 0>}, offsets=<ltrb 5 5 5 5>}}}
  }
  Border_0 {
   widgetClassPath=&Border
   parent=VerticalBox_0
   widget=Border_0
   properties {brushColor=<rgba 0 0 0 0.25>, padding=<ltrb 5 5 5 5>}
   slot {class=VerticalBoxSlot, properties {padding=<ltrb 0 5 0 5>}}
  }
  HorizontalBox_0 {
   widgetClassPath=&HorizontalBox
   parent=Border_0
   widget=HorizontalBox_0
   slot {class=BorderSlot, properties {padding=<ltrb 5 5 5 5>}}
  }
  ScaleBox_1 {
   widgetClassPath=&ScaleBox
   parent=HorizontalBox_0
   widget=ScaleBox_1
   properties {stretchDirection=DownOnly}
   slot {class=HorizontalBoxSlot, properties {verticalAlignment=VAlign_Center, size {sizeRule=Fill, value=1}}}
  }
  CommonTextBlock_72 {
   widgetClassPath=&CommonTextBlock
   bIsVariable=true
   parent=ScaleBox_1
   widget=CommonTextBlock_72
   properties {
    text=Text Block
    colorAndOpacity {colorUseRule=UseColor_Specified, specifiedColor=<rgba 0.901 0.901 0.901 1>}
    font {
     bForceMonospaced=false
     bMaterialIsStencil=false
     fontMaterial=null
     fontObject=&NotoSans
     letterSpacing=0
     monospacedWidth=1
     size=14
     skewAmount=0
     typefaceFontName=NotoSans-Regular
     outlineSettings {
      bApplyOutlineToDropShadows=false
      bMiteredCorners=false
      bSeparateFillAlpha=false
      outlineMaterial=null
      outlineSize=0
      outlineColor=<rgba 0 0 0 1>
     }
    }
    shadowOffset=<0 0>
   }
   slot {class=ScaleBoxSlot}
  }
  TA_Generic01_Button {
   widgetClassPath=&TA_Generic01_Button
   parent=HorizontalBox_0
   widget=TA_Generic01_Button
   properties {iconBrush {
     bIsDynamicallyLoaded=false
     drawAs=Image
     imageType=NoImage
     mirroring=NoMirror
     resourceName=None
     resourceObject=/Game/UI/Foundation/Textures/Icons/HUD/T-IconMagnifying-Glass-32
     tiling=NoTile
     imageSize=<16 16>
     margin=<ltrb 0 0 0 0>
     outlineSettings {
      bUseBrushTransparency=false
      roundingType=HalfHeightRadius
      width=0
      color {colorUseRule=UseColor_Specified, specifiedColor=<rgba 0 0 0 0>}
      cornerRadii=<0 0 0 1>
     }
     tintColor {colorUseRule=UseColor_Specified, specifiedColor=<rgba 1 1 1 1>}
     uVRegion {bIsValid=false, max=<0 0>, min=<0 0>}
    }}
   slot {class=HorizontalBoxSlot, properties {verticalAlignment=VAlign_Center}}
  }
  DetailsView_35 {
   widgetClassPath=&DetailsView
   parent=VerticalBox_0
   widget=DetailsView_35
   properties {
    bAllowFiltering=true
    bForceHiddenPropertyVisibility=true
    bShowAnimatedPropertiesOption=false
    bShowKeyablePropertiesOption=false
    bShowModifiedPropertiesOption=true
    propertiesToShow [
     LODForCollision
     SourceModels
     bAutoComputeLODScreenSize
     PerPlatform
     PercentTriangles
     PercentVertices
     WeldingThreshold
     LODDistance
     ScreenSize
     ReductionSettings
     TerminationCriterion
     MinLOD
     PlatformMinimumLODs
     BodySetup
     MinQualityLevelLOD
    ]
   }
   slot {class=VerticalBoxSlot, properties {size {sizeRule=Fill, value=1}}}
  }
  ScaleBox_0 {
   widgetClassPath=&ScaleBox
   parent=VerticalBox_0
   widget=ScaleBox_0
   properties {clipping=ClipToBounds}
   slot {class=VerticalBoxSlot}
  }
  CommonTextBlock {
   widgetClassPath=&CommonTextBlock
   bIsVariable=true
   parent=ScaleBox_0
   widget=CommonTextBlock
   properties {
    clipping=ClipToBounds
    text=All changes are automatically applied but you must click Save to save the asset.
    colorAndOpacity {colorUseRule=UseColor_Specified, specifiedColor=<rgba 0.7656 0.7656 0.7656 1>}
    font {
     bForceMonospaced=false
     bMaterialIsStencil=false
     fontMaterial=null
     fontObject=&NotoSans
     letterSpacing=0
     monospacedWidth=1
     size=10
     skewAmount=0
     typefaceFontName=NotoSans-Regular
     outlineSettings {
      bApplyOutlineToDropShadows=false
      bMiteredCorners=false
      bSeparateFillAlpha=false
      outlineMaterial=null
      outlineSize=0
      outlineColor=<rgba 0 0 0 1>
     }
    }
    shadowOffset=<0 0>
   }
   slot {class=ScaleBoxSlot}
  }
  btnSave {
   widgetClassPath=&TA_Generic01_Button
   parent=VerticalBox_0
   widget=btnSave
   properties {label=Save Asset}
   slot {class=VerticalBoxSlot, properties {padding=<ltrb 10 10 10 10>}}
  }
 ]
}
