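# Exclusion List popup widget for the artists' Asset Validation Tool, for pruning a list of excluded entries.
# openPopup fills a TA_Generic01_List from a string array and shows the widget; buttons remove the selected entry or clear all.
# Apply gathers the remaining display strings and broadcasts onPopupSubmitted, while Cancel broadcasts onPopupCancel.
# Both then clear the list and collapse the widget, leaving the caller to handle the result.
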WidgetBlueprint /Game/Blueprints/Blutilities/ForArtists/AssetValidationTool/SubWidgets/AVT_ExclusionList_Widget
graphs [
 EventGraph {nodes [
   "On Clicked (btnApply)"!() -> (OutputDelegate, then, widget:User Widget Object Reference, event:Pointer Event Structure) {is=title K2Node_ComponentBoundEvent}
   > handleApply()
   "On Clicked (btnRemoveAll)"!() -> (OutputDelegate, then, widget:User Widget Object Reference, event:Pointer Event Structure) {is=title K2Node_ComponentBoundEvent}
   > clearItems(self:TA Generic 01 List Object Reference=Get TA_Generic01_List() -> (TA_Generic01_List:TA Generic 01 List Object Reference))
   "On Clicked (btnRemoveItem)"!() -> (OutputDelegate, then, widget:User Widget Object Reference, event:Pointer Event Structure) {is=title K2Node_ComponentBoundEvent}
   > n1 = getSelectedItems!(execute, self:TA Generic 01 List Object Reference=Get TA_Generic01_List()) -> (then, isAnySelected:bool, items:Array of User Widget Object References, selectedData:Array of Object References)
   > Branch(Condition=@n1)
   > removeItems!(execute, self:TA Generic 01 List Object Reference=Get TA_Generic01_List(), itemWidgets:Array of User Widget Object References=@n1.items) -> (then)
   "On Clicked (btnCancel)"!() -> (OutputDelegate, then, widget:User Widget Object Reference, event:Pointer Event Structure) {is=title K2Node_ComponentBoundEvent}
   > Call On Popup Cancel!(execute:Exec, self:AVT Exclusion List Widget Object Reference) -> (then:Exec) {is=title K2Node_CallDelegate}
   > closePopup()
  ]}
 closePopup {nodes [
   closePopup() {is=title K2Node_FunctionEntry}
   > clearItems(self=Get TA_Generic01_List())
   > SetVisibility(InVisibility=Collapsed)
  ]}
 handleApply {nodes [
   handleApply() {is=title K2Node_FunctionEntry}
   > n1 = getListData!(execute, self:TA Generic 01 List Object Reference=Get TA_Generic01_List()) -> (then, data:Array of Object References)
   > n2 = For Each Loop(Array:Array of Object References=@n1) -> (Array Element:Object Reference)
   Call On Popup Submitted!(
    execute:Exec=@n2.Completed
    self:AVT Exclusion List Widget Object Reference
    strings:Array of Strings=Get l_strings() -> (l_strings:Array of Strings)
   ) -> (then:Exec) {is=title K2Node_CallDelegate}
   > closePopup() {is=title K2Node_CallFunction}
   n3 = Cast To TA_GenericListItemData!(execute=@n2.LoopBody, Object=@n2) -> (then, CastFailed, AsTA Generic List Item Data:TA Generic List Item Data Object Reference) {is=title K2Node_DynamicCast}
   > Add(TargetArray=Get l_strings(), NewItem=Get displayString!(self:TA Generic List Item Data Object Reference=@n3) -> (displayString:String)) {is=title K2Node_CallArrayFunction}
  ]}
 onPopupCancel {nodes [onPopupCancel() {is=title K2Node_FunctionEntry}]}
 onPopupSubmitted {nodes [onPopupSubmitted!() -> (then, strings:Array of Strings) {is=title K2Node_FunctionEntry}]}
 openPopup {nodes [
   n1 = openPopup!() -> (then, strings:Array of Strings) {is=title K2Node_FunctionEntry}
   > setItems!(execute, self, strings:Array of Strings=@n1) -> (then)
   > SetVisibility(InVisibility=SelfHitTestInvisible)
  ]}
 setItems {nodes [
   n1 = setItems!() -> (then, strings:Array of Strings) {is=title K2Node_FunctionEntry}
   > clearItems(self=Get TA_Generic01_List())
   > n2 = For Each Loop(Array:Array of Strings=@n1) -> (Array Element:String)
   > addItemGeneric!(execute, self:TA Generic 01 List Object Reference=Get TA_Generic01_List(), displayString:String=@n2) -> (then, widget:User Widget Object Reference, data:Object Reference)
  ]}
]
variables [
 onPopupCancel {}
 onPopupSubmitted {}
]
widgets {
 info {inheritedWidgetCount=0, parentClass=&UserWidget, rootWidgetClass=&Border, widgetCount=13}
 widgets [
  Border_5 {widgetClassPath=&Border, widget=Border_5, properties {brushColor=<rgba 0 0 0 0.8>, padding=<ltrb 40 20 40 40>}}
  VerticalBox_1 {
   widgetClassPath=&VerticalBox
   parent=Border_5
   widget=VerticalBox_1
   slot {class=BorderSlot, properties {padding=<ltrb 40 20 40 40>}}
  }
  CommonTextBlock_45 {
   widgetClassPath=&CommonTextBlock
   parent=VerticalBox_1
   widget=CommonTextBlock_45
   properties {
    style=&TextStyle-Base_C
    text=Exclusion List
    font {
     bForceMonospaced=false
     bMaterialIsStencil=false
     fontMaterial=null
     fontObject=&NotoSans
     letterSpacing=0
     monospacedWidth=1
     size=24
     skewAmount=0
     typefaceFontName=NotoSans-Regular
     outlineSettings {
      bApplyOutlineToDropShadows=false
      bMiteredCorners=false
      bSeparateFillAlpha=false
      outlineMaterial=null
      outlineSize=0
      outlineColor=<rgba 0 0 0 1>
     }
    }
    shadowOffset=<0 0>
   }
   slot {class=VerticalBoxSlot, properties {padding=<ltrb 20 0 0 0>}}
  }
  Border_0 {
   widgetClassPath=&Border
   parent=VerticalBox_1
   widget=Border_0
   properties {brushColor=<rgba 0.117 0.117 0.117 1>, padding=<ltrb 0 0 0 0>}
   slot {class=VerticalBoxSlot, properties {padding=<ltrb 20 10 20 10>, size {sizeRule=Fill, value=1}}}
  }
  VerticalBox_0 {widgetClassPath=&VerticalBox, parent=Border_0, widget=VerticalBox_0, slot {class=BorderSlot, properties {padding=<ltrb 0 0 0 0>}}}
  Border_1 {
   widgetClassPath=&Border
   parent=VerticalBox_0
   widget=Border_1
   properties {brushColor=<rgba 0.0482 0.0482 0.0482 1>, padding=<ltrb 5 5 5 5>}
   slot {class=VerticalBoxSlot, properties {padding=<ltrb 10 10 10 10>, size {sizeRule=Fill, value=1}}}
  }
  TA_Generic01_List {
   widgetClassPath=/Game/Blueprints/Blutilities/ForTechArt/Widgets/List/TA_Generic01_List
   bIsVariable=true
   parent=Border_1
   widget=TA_Generic01_List
   properties {allowMultiSelect=false, itemPadding=0}
   slot {class=BorderSlot, properties {padding=<ltrb 5 5 5 5>}}
  }
  HorizontalBox_1 {
   widgetClassPath=&HorizontalBox
   parent=VerticalBox_0
   widget=HorizontalBox_1
   slot {class=VerticalBoxSlot, properties {padding=<ltrb 10 10 10 10>}}
  }
  btnRemoveItem {
   widgetClassPath=&TA_Generic01_Button
   parent=HorizontalBox_1
   widget=btnRemoveItem
   properties {
    label=Remove Selected from Exclusion List
    toolTipText=Add selected list items to a list of meshes that will never show in the list again.
    buttonColor_Default=<rgba 1 1 1 1>
    buttonStyle_Default {
     clickedSlateSound {resourceObject=null}
     disabled {
      bIsDynamicallyLoaded=false
      drawAs=Box
      imageType=NoImage
      mirroring=NoMirror
      resourceName=None
      resourceObject=/Game/Blueprints/Blutilities/ForTechArt/ArtAssets/Textures/Button02/MI_TA_ButtonBG_02_Down
      tiling=NoTile
      imageSize=<32 32>
      margin=<ltrb 0.25 0.25 0.25 0.25>
      outlineSettings {
       bUseBrushTransparency=false
       roundingType=HalfHeightRadius
       width=0
       color {colorUseRule=UseColor_Specified, specifiedColor=<rgba 0 0 0 0>}
       cornerRadii=<0 0 0 1>
      }
      tintColor {colorUseRule=UseColor_Specified, specifiedColor=<rgba 0.3333 0.3333 0.3333 1>}
      uVRegion {bIsValid=false, max=<0 0>, min=<0 0>}
     }
     disabledForeground {colorUseRule=UseColor_Foreground, specifiedColor=<rgba 1 0 1 1>}
     hovered {
      bIsDynamicallyLoaded=false
      drawAs=Box
      imageType=NoImage
      mirroring=NoMirror
      resourceName=None
      resourceObject=/Game/Blueprints/Blutilities/ForTechArt/ArtAssets/Textures/Button02/MI_TA_ButtonBG_02_Over
      tiling=NoTile
      imageSize=<32 32>
      margin=<ltrb 0.25 0.25 0.25 0.25>
      outlineSettings {
       bUseBrushTransparency=false
       roundingType=HalfHeightRadius
       width=0
       color {colorUseRule=UseColor_Specified, specifiedColor=<rgba 0 0 0 0>}
       cornerRadii=<0 0 0 1>
      }
      tintColor {colorUseRule=UseColor_Specified, specifiedColor=<rgba 0.8542 0 0 1>}
      uVRegion {bIsValid=false, max=<0 0>, min=<0 0>}
     }
     hoveredForeground {colorUseRule=UseColor_Foreground, specifiedColor=<rgba 1 0 1 1>}
     hoveredSlateSound {resourceObject=null}
     normal {
      bIsDynamicallyLoaded=false
      drawAs=Box
      imageType=NoImage
      mirroring=NoMirror
      resourceName=None
      resourceObject=/Game/Blueprints/Blutilities/ForTechArt/ArtAssets/Textures/Button02/M_TA_ButtonBG_02_Up
      tiling=NoTile
      imageSize=<32 32>
      margin=<ltrb 0.25 0.25 0.25 0.25>
      outlineSettings {
       bUseBrushTransparency=false
       roundingType=HalfHeightRadius
       width=0
       color {colorUseRule=UseColor_Specified, specifiedColor=<rgba 0 0 0 0>}
       cornerRadii=<0 0 0 1>
      }
      tintColor {colorUseRule=UseColor_Specified, specifiedColor=<rgba 0.5457 0.0429 0.0103 1>}
      uVRegion {bIsValid=false, max=<0 0>, min=<0 0>}
     }
     normalForeground {colorUseRule=UseColor_Foreground, specifiedColor=<rgba 1 0 1 1>}
     normalPadding=<ltrb 0 0 0 0>
     pressed {
      bIsDynamicallyLoaded=false
      drawAs=Box
      imageType=NoImage
      mirroring=NoMirror
      resourceName=None
      resourceObject=&MI_TA_ButtonBG_02_Down
      tiling=NoTile
      imageSize=<32 32>
      margin=<ltrb 0.25 0.25 0.25 0.25>
      outlineSettings {
       bUseBrushTransparency=false
       roundingType=HalfHeightRadius
       width=0
       color {colorUseRule=UseColor_Specified, specifiedColor=<rgba 0 0 0 0>}
       cornerRadii=<0 0 0 1>
      }
      tintColor {colorUseRule=UseColor_Specified, specifiedColor=<rgba 0.2396 0 0 1>}
      uVRegion {bIsValid=false, max=<0 0>, min=<0 0>}
     }
     pressedForeground {colorUseRule=UseColor_Foreground, specifiedColor=<rgba 1 0 1 1>}
     pressedPadding=<ltrb 0 0 0 0>
     pressedSlateSound {resourceObject=null}
    }
    labelColor_Default=<rgba 0.9219 0.9219 0.9219 1>
   }
   slot {class=HorizontalBoxSlot, properties {size {sizeRule=Fill, value=1}}}
  }
  btnRemoveAll {
   widgetClassPath=&TA_Generic01_Button
   parent=HorizontalBox_1
   widget=btnRemoveAll
   properties {
    label=CLEAR LIST
    buttonColor_Default=<rgba 1 1 1 1>
    buttonStyle_Default {
     clickedSlateSound {resourceObject=null}
     disabled {
      bIsDynamicallyLoaded=false
      drawAs=Box
      imageType=NoImage
      mirroring=NoMirror
      resourceName=None
      resourceObject=&MI_TA_ButtonBG_02_Down
      tiling=NoTile
      imageSize=<32 32>
      margin=<ltrb 0.25 0.25 0.25 0.25>
      outlineSettings {
       bUseBrushTransparency=false
       roundingType=HalfHeightRadius
       width=0
       color {colorUseRule=UseColor_Specified, specifiedColor=<rgba 0 0 0 0>}
       cornerRadii=<0 0 0 1>
      }
      tintColor {colorUseRule=UseColor_Specified, specifiedColor=<rgba 0.3333 0.3333 0.3333 1>}
      uVRegion {bIsValid=false, max=<0 0>, min=<0 0>}
     }
     disabledForeground {colorUseRule=UseColor_Foreground, specifiedColor=<rgba 1 0 1 1>}
     hovered {
      bIsDynamicallyLoaded=false
      drawAs=Box
      imageType=NoImage
      mirroring=NoMirror
      resourceName=None
      resourceObject=&MI_TA_ButtonBG_02_Over
      tiling=NoTile
      imageSize=<32 32>
      margin=<ltrb 0.25 0.25 0.25 0.25>
      outlineSettings {
       bUseBrushTransparency=false
       roundingType=HalfHeightRadius
       width=0
       color {colorUseRule=UseColor_Specified, specifiedColor=<rgba 0 0 0 0>}
       cornerRadii=<0 0 0 1>
      }
      tintColor {colorUseRule=UseColor_Specified, specifiedColor=<rgba 0.8542 0 0 1>}
      uVRegion {bIsValid=false, max=<0 0>, min=<0 0>}
     }
     hoveredForeground {colorUseRule=UseColor_Foreground, specifiedColor=<rgba 1 0 1 1>}
     hoveredSlateSound {resourceObject=null}
     normal {
      bIsDynamicallyLoaded=false
      drawAs=Box
      imageType=NoImage
      mirroring=NoMirror
      resourceName=None
      resourceObject=&M_TA_ButtonBG_02_Up
      tiling=NoTile
      imageSize=<32 32>
      margin=<ltrb 0.25 0.25 0.25 0.25>
      outlineSettings {
       bUseBrushTransparency=false
       roundingType=HalfHeightRadius
       width=0
       color {colorUseRule=UseColor_Specified, specifiedColor=<rgba 0 0 0 0>}
       cornerRadii=<0 0 0 1>
      }
      tintColor {colorUseRule=UseColor_Specified, specifiedColor=<rgba 0.5457 0 0 1>}
      uVRegion {bIsValid=false, max=<0 0>, min=<0 0>}
     }
     normalForeground {colorUseRule=UseColor_Foreground, specifiedColor=<rgba 1 0 1 1>}
     normalPadding=<ltrb 0 0 0 0>
     pressed {
      bIsDynamicallyLoaded=false
      drawAs=Box
      imageType=NoImage
      mirroring=NoMirror
      resourceName=None
      resourceObject=&MI_TA_ButtonBG_02_Down
      tiling=NoTile
      imageSize=<32 32>
      margin=<ltrb 0.25 0.25 0.25 0.25>
      outlineSettings {
       bUseBrushTransparency=false
       roundingType=HalfHeightRadius
       width=0
       color {colorUseRule=UseColor_Specified, specifiedColor=<rgba 0 0 0 0>}
       cornerRadii=<0 0 0 1>
      }
      tintColor {colorUseRule=UseColor_Specified, specifiedColor=<rgba 0.2396 0 0 1>}
      uVRegion {bIsValid=false, max=<0 0>, min=<0 0>}
     }
     pressedForeground {colorUseRule=UseColor_Foreground, specifiedColor=<rgba 1 0 1 1>}
     pressedPadding=<ltrb 0 0 0 0>
     pressedSlateSound {resourceObject=null}
    }
    labelColor_Default=<rgba 0.9219 0.9219 0.9219 1>
   }
   slot {class=HorizontalBoxSlot, properties {padding=<ltrb 10 0 0 0>}}
  }
  HorizontalBox_2 {
   widgetClassPath=&HorizontalBox
   parent=VerticalBox_1
   widget=HorizontalBox_2
   slot {class=VerticalBoxSlot, properties {padding=<ltrb 30 5 30 10>}}
  }
  btnApply {
   widgetClassPath=&TA_Generic01_Button
   parent=HorizontalBox_2
   widget=btnApply
   properties {label=Apply and Close}
   slot {class=HorizontalBoxSlot, properties {size {sizeRule=Fill, value=1}}}
  }
  btnCancel {
   widgetClassPath=&TA_Generic01_Button
   parent=HorizontalBox_2
   widget=btnCancel
   properties {label=Cancel and Close}
   slot {class=HorizontalBoxSlot, properties {padding=<ltrb 10 0 0 0>, size {sizeRule=Fill, value=1}}}
  }
 ]
}
