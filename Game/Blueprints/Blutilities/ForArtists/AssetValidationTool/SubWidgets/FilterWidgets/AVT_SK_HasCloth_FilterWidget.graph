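# Editor Utility filter widget for the Asset Validation Tool that flags skeletal meshes with cloth assigned.
# filterAsset loops the mesh's materials, tests each index with IsSectionUsingCloth, and returns a "Has Cloth" error struct on the first hit.
# A "Has Cloth" checkbox toggles the filter; validateFilterData stores its state in the base class's filter data, and validateWidgets restores it.
# The checkbox tooltip mistakenly describes a LOD check.
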EditorUtilityWidgetBlueprint /Game/Blueprints/Blutilities/ForArtists/AssetValidationTool/SubWidgets/FilterWidgets/AVT_SK_HasCloth_FilterWidget
graphs [
 EventGraph {nodes [
   n1 = "On Check State Changed (checkBox)"!() -> (OutputDelegate, then, bIsChecked:bool) {is=title K2Node_ComponentBoundEvent}
   > setFilterEnabled!(execute, self:AVT Base Filter Widget Object Reference, isEnabled:bool=@n1.bIsChecked) -> (then, changed:bool)
  ]}
 filterAsset {nodes [
   n1 = filterAsset!() -> (then, asset:Object Reference) {is=title K2Node_FunctionEntry}
   > n2 = Cast To SkeletalMesh!(execute, Object=@n1) -> (then, CastFailed, AsSkeletal Mesh:Skeletal Mesh Object Reference) {is=title K2Node_DynamicCast}
   > n3 = For Each Loop(Array:Array of Skeletal Material Structures=Get Materials!(self:Skeletal Mesh Object Reference=@n2) -> (Materials:Array of Skeletal Material Structures)) -> (Array Element:Skeletal Material Structure)
   Return Node!(execute=@n3.Completed, hasError:bool=false, ErrorStruct:AVT Error Struct Structure)
   > Branch(Condition=IsSectionUsingCloth!(self:Skeletal Mesh Object Reference=@n2, InSectionIndex:int=@n3.Array Index, bCheckCorrespondingSections:bool=true) -> (ReturnValue:bool))
   > Return Node(
    hasError=true
    ErrorStruct=Make AVT Error Struct!(
     title_32_#1:String=Cloth
     description_35_#2:String=Has Cloth
     tooltip_37_#3:String=This asset has cloth assigned to a section.
     color_40_#4:Linear Color Structure=(R=0.9635,G=0.0396,B=0.0301,A=1)
     csvLine_46_#5:String=True
     csvHeader_49_#6:String=Has Cloth Assigned
     optionalAssetData_52_#7:Asset Data Structure=()
     optionalTypeTag_55_#8:String
     optionalOriginPackageName_62_#9:Name=None
     helpCommand_65_#10:String
    ) -> (AVT_Error_Struct:AVT Error Struct Structure) {is=title K2Node_MakeStruct}
   )
  ]}
 validateFilterData {nodes [
   validateFilterData() {is=title K2Node_FunctionEntry}
   > Parent: validateFilterData!(execute) -> (then, " ":bool) {is=title K2Node_CallParentFunction}
   n1 = Get checkBox() -> (checkBox:Check Box Object Reference)
   > Set members in AVT Filter Data Struct!(
    execute:Exec
    StructRef:"AVT Filter Data Struct Structure (by ref)"=Get currentFilterData() -> (currentFilterData:AVT Filter Data Struct Structure)
    Bools_2_#11:Array of Booleans=Make Array("[0]":bool=IsChecked(self=@n1)) -> (Array:Array of Booleans)
   ) -> (then:Exec, StructOut:"AVT Filter Data Struct Structure (by ref)") {is=title K2Node_SetFieldsInStruct}
   Return Node!(execute, " ":bool=false)
  ]}
 validateWidgets {nodes [
   validateWidgets() {is=title K2Node_FunctionEntry}
   > Branch(Condition="Not Equal (Name)"!(
     A:Name=Get!() -> (
      currentFilterData_Bools_2_#11:Array of Booleans
      currentFilterData_Strings_5_#12:Array of Strings
      currentFilterData_Floats_8_#13:"Array of Float (single-precision)s"
      currentFilterData_filterID_13_#14:Name
      currentFilterData_filterEnabled_15_#15:bool=~
      currentFilterData_hasHadDataSet_17_#16:bool=~
     ).currentFilterData_filterID_13_#14
     B:Name=None
    ) -> (ReturnValue:bool) {is=title K2Node_CallFunction})
   > Parent: validateWidgets!(execute) -> (then, " ":bool) {is=title K2Node_CallParentFunction}
   > SetIsChecked!(execute, self:Check Box Object Reference=Get checkBox(), InIsChecked:bool=getFilterEnabled!() -> (isEnabled:bool)) -> (then)
   Return Node(" "=false)
  ]}
]
variables []
widgets {
 info {
  parentClass=&/AVT_Base_FilterWidget.AVT_Base_FilterWidget_C
  rootWidgetClass=&CanvasPanel
  widgetCount=6
 }
 widgets [
  CanvasPanel_0 {widgetClassPath=&CanvasPanel, widget=CanvasPanel_0}
  SizeBox_0 {
   widgetClassPath=&SizeBox
   parent=CanvasPanel_0
   widget=SizeBox_0
   properties {bOverride_HeightOverride=true, heightOverride=30}
   slot {
    class=CanvasPanelSlot
    properties {bAutoSize=true, layoutData {alignment=<0 0>, anchors {maximum=<0 0>, minimum=<0 0>}, offsets=<ltrb 0 0 280.6229 30>}}
   }
  }
  Border_4 {
   widgetClassPath=&Border
   parent=SizeBox_0
   widget=Border_4
   properties {brushColor=<rgba 0.0296 0.0296 0.0296 1>, padding=<ltrb 3 3 3 3>}
   slot {class=SizeBoxSlot}
  }
  HorizontalBox_0 {
   widgetClassPath=&HorizontalBox
   parent=Border_4
   widget=HorizontalBox_0
   slot {class=BorderSlot, properties {padding=<ltrb 3 3 3 3>}}
  }
  checkBox {
   widgetClassPath=/Script/UMG.CheckBox
   bIsVariable=true
   parent=HorizontalBox_0
   widget=checkBox
   properties {
    toolTipText=Check if meshes have any LODs at all.
    widgetStyle {
     checkBoxType=CheckBox
     backgroundHoveredImage {
      bIsDynamicallyLoaded=false
      drawAs=RoundedBox
      imageType=NoImage
      mirroring=NoMirror
      resourceName=None
      resourceObject=null
      tiling=NoTile
      imageSize=<18 18>
      margin=<ltrb 0 0 0 0>
      outlineSettings {
       bUseBrushTransparency=false
       roundingType=FixedRadius
       width=1
       color {colorUseRule=UseColor_ColorTable, specifiedColor=<rgba 0 0 0 0>}
       cornerRadii=<3 3 3 3>
      }
      tintColor {colorUseRule=UseColor_Specified, specifiedColor=<rgba 0.0048 0.0048 0.0048 1>}
      uVRegion {bIsValid=false, max=<0 0>, min=<0 0>}
     }
     backgroundImage {
      bIsDynamicallyLoaded=false
      drawAs=RoundedBox
      imageType=NoImage
      mirroring=NoMirror
      resourceName=None
      resourceObject=null
      tiling=NoTile
      imageSize=<18 18>
      margin=<ltrb 0 0 0 0>
      outlineSettings {
       bUseBrushTransparency=false
       roundingType=FixedRadius
       width=1
       color {colorUseRule=UseColor_ColorTable, specifiedColor=<rgba 0 0 0 0>}
       cornerRadii=<3 3 3 3>
      }
      tintColor {colorUseRule=UseColor_Specified, specifiedColor=<rgba 0.0048 0.0048 0.0048 1>}
      uVRegion {bIsValid=false, max=<0 0>, min=<0 0>}
     }
     backgroundPressedImage {
      bIsDynamicallyLoaded=false
      drawAs=RoundedBox
      imageType=NoImage
      mirroring=NoMirror
      resourceName=None
      resourceObject=null
      tiling=NoTile
      imageSize=<18 18>
      margin=<ltrb 0 0 0 0>
      outlineSettings {
       bUseBrushTransparency=false
       roundingType=FixedRadius
       width=1
       color {colorUseRule=UseColor_ColorTable, specifiedColor=<rgba 0 0 0 0>}
       cornerRadii=<3 3 3 3>
      }
      tintColor {colorUseRule=UseColor_Specified, specifiedColor=<rgba 0.0048 0.0048 0.0048 1>}
      uVRegion {bIsValid=false, max=<0 0>, min=<0 0>}
     }
     borderBackgroundColor {colorUseRule=UseColor_Specified, specifiedColor=<rgba 1 1 1 1>}
     checkedForeground {colorUseRule=UseColor_Specified, specifiedColor=<rgba 1 1 1 1>}
     checkedHoveredForeground {colorUseRule=UseColor_Specified, specifiedColor=<rgba 1 1 1 1>}
     checkedHoveredImage {
      bIsDynamicallyLoaded=false
      drawAs=Image
      imageType=Vector
      mirroring=NoMirror
      resourceName=../../../Engine/Content/Slate/Starship/CoreWidgets/CheckBox/check.svg
      resourceObject=null
      tiling=NoTile
      imageSize=<16 16>
      margin=<ltrb 0 0 0 0>
      outlineSettings {
       bUseBrushTransparency=false
       roundingType=HalfHeightRadius
       width=0
       color {colorUseRule=UseColor_Specified, specifiedColor=<rgba 0 0 0 0>}
       cornerRadii=<0 0 0 0>
      }
      tintColor {colorUseRule=UseColor_Specified, specifiedColor=<rgba 0.0044 0.2384 1 1>}
      uVRegion {bIsValid=false, max=<0 0>, min=<0 0>}
     }
     checkedImage {
      bIsDynamicallyLoaded=false
      drawAs=Image
      imageType=Vector
      mirroring=NoMirror
      resourceName=../../../Engine/Content/Slate/Starship/CoreWidgets/CheckBox/check.svg
      resourceObject=null
      tiling=NoTile
      imageSize=<16 16>
      margin=<ltrb 0 0 0 0>
      outlineSettings {
       bUseBrushTransparency=false
       roundingType=HalfHeightRadius
       width=0
       color {colorUseRule=UseColor_Specified, specifiedColor=<rgba 0 0 0 0>}
       cornerRadii=<0 0 0 0>
      }
      tintColor {colorUseRule=UseColor_Specified, specifiedColor=<rgba 0 0.162 0.7454 1>}
      uVRegion {bIsValid=false, max=<0 0>, min=<0 0>}
     }
     checkedPressedForeground {colorUseRule=UseColor_Specified, specifiedColor=<rgba 1 1 1 1>}
     checkedPressedImage {
      bIsDynamicallyLoaded=false
      drawAs=Image
      imageType=Vector
      mirroring=NoMirror
      resourceName=../../../Engine/Content/Slate/Starship/CoreWidgets/CheckBox/check.svg
      resourceObject=null
      tiling=NoTile
      imageSize=<16 16>
      margin=<ltrb 0 0 0 0>
      outlineSettings {
       bUseBrushTransparency=false
       roundingType=HalfHeightRadius
       width=0
       color {colorUseRule=UseColor_Specified, specifiedColor=<rgba 0 0 0 0>}
       cornerRadii=<0 0 0 0>
      }
      tintColor {colorUseRule=UseColor_Specified, specifiedColor=<rgba 0 0.0802 0.3515 1>}
      uVRegion {bIsValid=false, max=<0 0>, min=<0 0>}
     }
     checkedSlateSound {resourceObject=null}
     foregroundColor {colorUseRule=UseColor_Specified, specifiedColor=<rgba 1 1 1 1>}
     hoveredForeground {colorUseRule=UseColor_Specified, specifiedColor=<rgba 1 1 1 1>}
     hoveredSlateSound {resourceObject=null}
     padding=<ltrb 2 0 0 0>
     pressedForeground {colorUseRule=UseColor_Specified, specifiedColor=<rgba 1 1 1 1>}
     uncheckedHoveredImage {
      bIsDynamicallyLoaded=false
      drawAs=NoDrawType
      imageType=NoImage
      mirroring=NoMirror
      resourceName=None
      resourceObject=null
      tiling=NoTile
      imageSize=<0 0>
      margin=<ltrb 0 0 0 0>
      outlineSettings {
       bUseBrushTransparency=false
       roundingType=HalfHeightRadius
       width=0
       color {colorUseRule=UseColor_Specified, specifiedColor=<rgba 0 0 0 0>}
       cornerRadii=<0 0 0 0>
      }
      tintColor {colorUseRule=UseColor_Specified, specifiedColor=<rgba 1 1 1 1>}
      uVRegion {bIsValid=false, max=<0 0>, min=<0 0>}
     }
     uncheckedImage {
      bIsDynamicallyLoaded=false
      drawAs=NoDrawType
      imageType=NoImage
      mirroring=NoMirror
      resourceName=None
      resourceObject=null
      tiling=NoTile
      imageSize=<0 0>
      margin=<ltrb 0 0 0 0>
      outlineSettings {
       bUseBrushTransparency=false
       roundingType=HalfHeightRadius
       width=0
       color {colorUseRule=UseColor_Specified, specifiedColor=<rgba 0 0 0 0>}
       cornerRadii=<0 0 0 0>
      }
      tintColor {colorUseRule=UseColor_Specified, specifiedColor=<rgba 1 1 1 1>}
      uVRegion {bIsValid=false, max=<0 0>, min=<0 0>}
     }
     uncheckedPressedImage {
      bIsDynamicallyLoaded=false
      drawAs=NoDrawType
      imageType=NoImage
      mirroring=NoMirror
      resourceName=None
      resourceObject=null
      tiling=NoTile
      imageSize=<0 0>
      margin=<ltrb 0 0 0 0>
      outlineSettings {
       bUseBrushTransparency=false
       roundingType=HalfHeightRadius
       width=0
       color {colorUseRule=UseColor_Specified, specifiedColor=<rgba 0 0 0 0>}
       cornerRadii=<0 0 0 0>
      }
      tintColor {colorUseRule=UseColor_Specified, specifiedColor=<rgba 1 1 1 1>}
      uVRegion {bIsValid=false, max=<0 0>, min=<0 0>}
     }
     uncheckedSlateSound {resourceObject=null}
     undeterminedForeground {colorUseRule=UseColor_Specified, specifiedColor=<rgba 1 1 1 1>}
     undeterminedHoveredImage {
      bIsDynamicallyLoaded=false
      drawAs=Image
      imageType=Vector
      mirroring=NoMirror
      resourceName=../../../Engine/Content/Slate/Starship/CoreWidgets/CheckBox/indeterminate.svg
      resourceObject=null
      tiling=NoTile
      imageSize=<16 16>
      margin=<ltrb 0 0 0 0>
      outlineSettings {
       bUseBrushTransparency=false
       roundingType=HalfHeightRadius
       width=0
       color {colorUseRule=UseColor_Specified, specifiedColor=<rgba 0 0 0 0>}
       cornerRadii=<0 0 0 0>
      }
      tintColor {colorUseRule=UseColor_Specified, specifiedColor=<rgba 0.0044 0.2384 1 1>}
      uVRegion {bIsValid=false, max=<0 0>, min=<0 0>}
     }
     undeterminedImage {
      bIsDynamicallyLoaded=false
      drawAs=Image
      imageType=Vector
      mirroring=NoMirror
      resourceName=../../../Engine/Content/Slate/Starship/CoreWidgets/CheckBox/indeterminate.svg
      resourceObject=null
      tiling=NoTile
      imageSize=<16 16>
      margin=<ltrb 0 0 0 0>
      outlineSettings {
       bUseBrushTransparency=false
       roundingType=HalfHeightRadius
       width=0
       color {colorUseRule=UseColor_Specified, specifiedColor=<rgba 0 0 0 0>}
       cornerRadii=<0 0 0 0>
      }
      tintColor {colorUseRule=UseColor_Specified, specifiedColor=<rgba 0 0.162 0.7454 1>}
      uVRegion {bIsValid=false, max=<0 0>, min=<0 0>}
     }
     undeterminedPressedImage {
      bIsDynamicallyLoaded=false
      drawAs=Image
      imageType=Vector
      mirroring=NoMirror
      resourceName=../../../Engine/Content/Slate/Starship/CoreWidgets/CheckBox/indeterminate.svg
      resourceObject=null
      tiling=NoTile
      imageSize=<16 16>
      margin=<ltrb 0 0 0 0>
      outlineSettings {
       bUseBrushTransparency=false
       roundingType=HalfHeightRadius
       width=0
       color {colorUseRule=UseColor_Specified, specifiedColor=<rgba 0 0 0 0>}
       cornerRadii=<0 0 0 0>
      }
      tintColor {colorUseRule=UseColor_Specified, specifiedColor=<rgba 0 0.0802 0.3515 1>}
      uVRegion {bIsValid=false, max=<0 0>, min=<0 0>}
     }
    }
   }
   slot {class=HorizontalBoxSlot, properties {verticalAlignment=VAlign_Center}}
  }
  TextBlock_2 {
   widgetClassPath=&TextBlock
   parent=HorizontalBox_0
   widget=TextBlock_2
   properties {
    text=Has Cloth
    font {
     bForceMonospaced=false
     bMaterialIsStencil=false
     fontMaterial=null
     fontObject=&Roboto
     letterSpacing=0
     monospacedWidth=1
     size=9
     skewAmount=0
     typefaceFontName=Bold
     outlineSettings {
      bApplyOutlineToDropShadows=false
      bMiteredCorners=false
      bSeparateFillAlpha=false
      outlineMaterial=null
      outlineSize=0
      outlineColor=<rgba 0 0 0 1>
     }
    }
   }
   slot {class=HorizontalBoxSlot, properties {verticalAlignment=VAlign_Center}}
  }
 ]
}
