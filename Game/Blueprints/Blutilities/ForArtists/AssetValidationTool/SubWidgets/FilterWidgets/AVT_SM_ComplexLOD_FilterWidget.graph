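# Editor utility filter widget for the Asset Validation Tool that flags static meshes lacking a Complex Collision LOD.
# filterAsset reports a "No Collision LOD" error when LODForCollision is below 1, collision complexity differs from UseDefault, and any LOD 0 section has collision.
# A helper loop over LOD 0 sections stops at the first with collision enabled.
# The checkbox toggles the filter, and overrides sync its state with the base class's filter data.
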EditorUtilityWidgetBlueprint /Game/Blueprints/Blutilities/ForArtists/AssetValidationTool/SubWidgets/FilterWidgets/AVT_SM_ComplexLOD_FilterWidget
graphs [
 EventGraph {nodes [
   n1 = "On Check State Changed (checkBox)"!() -> (OutputDelegate, then, bIsChecked:bool) {is=title K2Node_ComponentBoundEvent}
   > setFilterEnabled!(execute, self:AVT Base Filter Widget Object Reference, isEnabled:bool=@n1.bIsChecked) -> (then, changed:bool)
  ]}
 filterAsset {nodes [
   n1 = filterAsset!() -> (then, asset:Object Reference) {is=title K2Node_FunctionEntry}
   > n2 = Cast To StaticMesh(Object=@n1)
   > Set l_mesh(l_mesh:Static Mesh Object Reference=@n2) -> (Output_Get:Static Mesh Object Reference)
   n3 = Get l_mesh() -> (l_mesh:Static Mesh Object Reference)
   > Branch(Condition="integer < integer"!(A:int=Get LODForCollision!(self:Static Mesh Object Reference=@n3) -> (LODForCollision:int=0), B:int=1) -> (ReturnValue:bool) {is=title K2Node_CallFunction})
   n4 = Get StaticMeshEditorSubsystem()
   n5 = Get l_mesh()
   > Branch(Condition="Not Equal (Enum)"!(
     A:ECollisionTraceFlag Enum=GetCollisionComplexity!(self:Static Mesh Editor Subsystem Object Reference=@n4, StaticMesh:Static Mesh Object Reference=@n5) -> (ReturnValue:ECollisionTraceFlag Enum=CTF_UseDefault)
     B:ECollisionTraceFlag Enum=CTF_UseDefault
    ) -> (ReturnValue:bool))
   > n6 = getIfMeshHasCollisionOnSections!(execute, self, meshToCheck:Static Mesh Object Reference=Get l_mesh()) -> (then, hasSectionWithCollision:bool)
   > Branch(Condition=@n6)
   > Return Node!(
    execute
    hasError:bool=true
    ErrorStruct:AVT Error Struct Structure=Make AVT Error Struct!(
     title_32_#1:String=Collision
     description_35_#2:String=No Collision LOD
     tooltip_37_#3:String=This mesh has complex collision and no Complex Collision LOD set
     color_40_#4:Linear Color Structure=lerpErrorColoring!(Alpha:double=1) -> (color:Linear Color Structure)
     csvLine_46_#5:String=No
     csvHeader_49_#6:String=Has Collision LOD
     optionalAssetData_52_#7:Asset Data Structure=()
     optionalTypeTag_55_#8:String
     optionalOriginPackageName_62_#9:Name=None
     helpCommand_65_#10:String
    ) -> (AVT_Error_Struct:AVT Error Struct Structure) {is=title K2Node_MakeStruct}
   )
  ]}
 getIfMeshHasCollisionOnSections {nodes [
   getIfMeshHasCollisionOnSections!() -> (then, meshToCheck:Static Mesh Object Reference) {is=title K2Node_FunctionEntry}
   > Set l_hadASectionWithCollision(l_hadASectionWithCollision:bool=false) -> (Output_Get:bool)
   n1 = Get meshToCheck() -> (meshToCheck:Static Mesh Object Reference)
   > n2 = For Loop with Break!(
    Execute:Exec
    FirstIndex:int
    LastIndex:int=int - int(A=GetNumSections!(self:Static Mesh Object Reference=@n1, InLOD:int=0) -> (ReturnValue:int=0), B=1) {is=title K2Node_CallFunction}
    Break:Exec=@n3.then
   ) -> (LoopBody:Exec, Index:int, Completed:Exec) {is=title K2Node_MacroInstance}
   Return Node!(execute=@n2.Completed, hasSectionWithCollision:bool=Get l_hadASectionWithCollision() -> (l_hadASectionWithCollision:bool))
   > Branch(Condition=IsSectionCollisionEnabled!(
     self:Static Mesh Editor Subsystem Object Reference=Get StaticMeshEditorSubsystem()
     StaticMesh:Static Mesh Object Reference=Get meshToCheck()
     LODIndex:int=0
     SectionIndex:int=@n2
    ) -> (ReturnValue:bool))
   > n3 = Set l_hadASectionWithCollision(l_hadASectionWithCollision=true)
  ]}
 validateFilterData {nodes [
   validateFilterData() {is=title K2Node_FunctionEntry}
   > Parent: validateFilterData!(execute) -> (then, " ":bool) {is=title K2Node_CallParentFunction}
   n1 = Get checkBox() -> (checkBox:Check Box Object Reference)
   > Set members in AVT Filter Data Struct!(
    execute:Exec
    StructRef:"AVT Filter Data Struct Structure (by ref)"=Get currentFilterData() -> (currentFilterData:AVT Filter Data Struct Structure)
    Bools_2_#11:Array of Booleans=Make Array("[0]":bool=IsChecked(self=@n1)) -> (Array:Array of Booleans)
   ) -> (then:Exec, StructOut:"AVT Filter Data Struct Structure (by ref)") {is=title K2Node_SetFieldsInStruct}
   Return Node!(execute, " ":bool=false)
  ]}
 validateWidgets {nodes [
   validateWidgets() {is=title K2Node_FunctionEntry}
   > Parent: validateWidgets!(execute) -> (then, " ":bool) {is=title K2Node_CallParentFunction}
   > SetIsChecked!(execute, self:Check Box Object Reference=Get checkBox(), InIsChecked:bool=getFilterEnabled!() -> (isEnabled:bool)) -> (then)
   Return Node(" "=false)
  ]}
]
variables []
widgets {
 info {
  parentClass=&/AVT_Base_FilterWidget.AVT_Base_FilterWidget_C
  rootWidgetClass=&CanvasPanel
  widgetCount=4
 }
 widgets [
  CanvasPanel_0 {widgetClassPath=&CanvasPanel, widget=CanvasPanel_0}
  Border_4 {
   widgetClassPath=&Border
   parent=CanvasPanel_0
   widget=Border_4
   properties {brushColor=<rgba 0.0296 0.0296 0.0296 1>, padding=<ltrb 3 3 3 3>}
   slot {class=CanvasPanelSlot, properties {layoutData {alignment=<0 0>, anchors {maximum=<0 0>, minimum=<0 0>}, offsets=<ltrb 0 0 119.5195 30>}}}
  }
  checkBox {
   widgetClassPath=/Script/UMG.CheckBox
   bIsVariable=true
   parent=Border_4
   widget=checkBox
   properties {
    checkedState=Checked
    toolTipText=Check if meshes have any LODs at all.
    widgetStyle {
     checkBoxType=CheckBox
     backgroundHoveredImage {
      bIsDynamicallyLoaded=false
      drawAs=RoundedBox
      imageType=NoImage
      mirroring=NoMirror
      resourceName=None
      resourceObject=null
      tiling=NoTile
      imageSize=<18 18>
      margin=<ltrb 0 0 0 0>
      outlineSettings {
       bUseBrushTransparency=false
       roundingType=FixedRadius
       width=1
       color {colorUseRule=UseColor_ColorTable, specifiedColor=<rgba 0 0 0 0>}
       cornerRadii=<3 3 3 3>
      }
      tintColor {colorUseRule=UseColor_Specified, specifiedColor=<rgba 0.0048 0.0048 0.0048 1>}
      uVRegion {bIsValid=false, max=<0 0>, min=<0 0>}
     }
     backgroundImage {
      bIsDynamicallyLoaded=false
      drawAs=RoundedBox
      imageType=NoImage
      mirroring=NoMirror
      resourceName=None
      resourceObject=null
      tiling=NoTile
      imageSize=<18 18>
      margin=<ltrb 0 0 0 0>
      outlineSettings {
       bUseBrushTransparency=false
       roundingType=FixedRadius
       width=1
       color {colorUseRule=UseColor_ColorTable, specifiedColor=<rgba 0 0 0 0>}
       cornerRadii=<3 3 3 3>
      }
      tintColor {colorUseRule=UseColor_Specified, specifiedColor=<rgba 0.0048 0.0048 0.0048 1>}
      uVRegion {bIsValid=false, max=<0 0>, min=<0 0>}
     }
     backgroundPressedImage {
      bIsDynamicallyLoaded=false
      drawAs=RoundedBox
      imageType=NoImage
      mirroring=NoMirror
      resourceName=None
      resourceObject=null
      tiling=NoTile
      imageSize=<18 18>
      margin=<ltrb 0 0 0 0>
      outlineSettings {
       bUseBrushTransparency=false
       roundingType=FixedRadius
       width=1
       color {colorUseRule=UseColor_ColorTable, specifiedColor=<rgba 0 0 0 0>}
       cornerRadii=<3 3 3 3>
      }
      tintColor {colorUseRule=UseColor_Specified, specifiedColor=<rgba 0.0048 0.0048 0.0048 1>}
      uVRegion {bIsValid=false, max=<0 0>, min=<0 0>}
     }
     borderBackgroundColor {colorUseRule=UseColor_Specified, specifiedColor=<rgba 1 1 1 1>}
     checkedForeground {colorUseRule=UseColor_Specified, specifiedColor=<rgba 1 1 1 1>}
     checkedHoveredForeground {colorUseRule=UseColor_Specified, specifiedColor=<rgba 1 1 1 1>}
     checkedHoveredImage {
      bIsDynamicallyLoaded=false
      drawAs=Image
      imageType=Vector
      mirroring=NoMirror
      resourceName=../../../Engine/Content/Slate/Starship/CoreWidgets/CheckBox/check.svg
      resourceObject=null
      tiling=NoTile
      imageSize=<16 16>
      margin=<ltrb 0 0 0 0>
      outlineSettings {
       bUseBrushTransparency=false
       roundingType=HalfHeightRadius
       width=0
       color {colorUseRule=UseColor_Specified, specifiedColor=<rgba 0 0 0 0>}
       cornerRadii=<0 0 0 0>
      }
      tintColor {colorUseRule=UseColor_Specified, specifiedColor=<rgba 0.0044 0.2384 1 1>}
      uVRegion {bIsValid=false, max=<0 0>, min=<0 0>}
     }
     checkedImage {
      bIsDynamicallyLoaded=false
      drawAs=Image
      imageType=Vector
      mirroring=NoMirror
      resourceName=../../../Engine/Content/Slate/Starship/CoreWidgets/CheckBox/check.svg
      resourceObject=null
      tiling=NoTile
      imageSize=<16 16>
      margin=<ltrb 0 0 0 0>
      outlineSettings {
       bUseBrushTransparency=false
       roundingType=HalfHeightRadius
       width=0
       color {colorUseRule=UseColor_Specified, specifiedColor=<rgba 0 0 0 0>}
       cornerRadii=<0 0 0 0>
      }
      tintColor {colorUseRule=UseColor_Specified, specifiedColor=<rgba 0 0.162 0.7454 1>}
      uVRegion {bIsValid=false, max=<0 0>, min=<0 0>}
     }
     checkedPressedForeground {colorUseRule=UseColor_Specified, specifiedColor=<rgba 1 1 1 1>}
     checkedPressedImage {
      bIsDynamicallyLoaded=false
      drawAs=Image
      imageType=Vector
      mirroring=NoMirror
      resourceName=../../../Engine/Content/Slate/Starship/CoreWidgets/CheckBox/check.svg
      resourceObject=null
      tiling=NoTile
      imageSize=<16 16>
      margin=<ltrb 0 0 0 0>
      outlineSettings {
       bUseBrushTransparency=false
       roundingType=HalfHeightRadius
       width=0
       color {colorUseRule=UseColor_Specified, specifiedColor=<rgba 0 0 0 0>}
       cornerRadii=<0 0 0 0>
      }
      tintColor {colorUseRule=UseColor_Specified, specifiedColor=<rgba 0 0.0802 0.3515 1>}
      uVRegion {bIsValid=false, max=<0 0>, min=<0 0>}
     }
     checkedSlateSound {resourceObject=null}
     foregroundColor {colorUseRule=UseColor_Specified, specifiedColor=<rgba 1 1 1 1>}
     hoveredForeground {colorUseRule=UseColor_Specified, specifiedColor=<rgba 1 1 1 1>}
     hoveredSlateSound {resourceObject=null}
     padding=<ltrb 2 0 0 0>
     pressedForeground {colorUseRule=UseColor_Specified, specifiedColor=<rgba 1 1 1 1>}
     uncheckedHoveredImage {
      bIsDynamicallyLoaded=false
      drawAs=NoDrawType
      imageType=NoImage
      mirroring=NoMirror
      resourceName=None
      resourceObject=null
      tiling=NoTile
      imageSize=<0 0>
      margin=<ltrb 0 0 0 0>
      outlineSettings {
       bUseBrushTransparency=false
       roundingType=HalfHeightRadius
       width=0
       color {colorUseRule=UseColor_Specified, specifiedColor=<rgba 0 0 0 0>}
       cornerRadii=<0 0 0 0>
      }
      tintColor {colorUseRule=UseColor_Specified, specifiedColor=<rgba 1 1 1 1>}
      uVRegion {bIsValid=false, max=<0 0>, min=<0 0>}
     }
     uncheckedImage {
      bIsDynamicallyLoaded=false
      drawAs=NoDrawType
      imageType=NoImage
      mirroring=NoMirror
      resourceName=None
      resourceObject=null
      tiling=NoTile
      imageSize=<0 0>
      margin=<ltrb 0 0 0 0>
      outlineSettings {
       bUseBrushTransparency=false
       roundingType=HalfHeightRadius
       width=0
       color {colorUseRule=UseColor_Specified, specifiedColor=<rgba 0 0 0 0>}
       cornerRadii=<0 0 0 0>
      }
      tintColor {colorUseRule=UseColor_Specified, specifiedColor=<rgba 1 1 1 1>}
      uVRegion {bIsValid=false, max=<0 0>, min=<0 0>}
     }
     uncheckedPressedImage {
      bIsDynamicallyLoaded=false
      drawAs=NoDrawType
      imageType=NoImage
      mirroring=NoMirror
      resourceName=None
      resourceObject=null
      tiling=NoTile
      imageSize=<0 0>
      margin=<ltrb 0 0 0 0>
      outlineSettings {
       bUseBrushTransparency=false
       roundingType=HalfHeightRadius
       width=0
       color {colorUseRule=UseColor_Specified, specifiedColor=<rgba 0 0 0 0>}
       cornerRadii=<0 0 0 0>
      }
      tintColor {colorUseRule=UseColor_Specified, specifiedColor=<rgba 1 1 1 1>}
      uVRegion {bIsValid=false, max=<0 0>, min=<0 0>}
     }
     uncheckedSlateSound {resourceObject=null}
     undeterminedForeground {colorUseRule=UseColor_Specified, specifiedColor=<rgba 1 1 1 1>}
     undeterminedHoveredImage {
      bIsDynamicallyLoaded=false
      drawAs=Image
      imageType=Vector
      mirroring=NoMirror
      resourceName=../../../Engine/Content/Slate/Starship/CoreWidgets/CheckBox/indeterminate.svg
      resourceObject=null
      tiling=NoTile
      imageSize=<16 16>
      margin=<ltrb 0 0 0 0>
      outlineSettings {
       bUseBrushTransparency=false
       roundingType=HalfHeightRadius
       width=0
       color {colorUseRule=UseColor_Specified, specifiedColor=<rgba 0 0 0 0>}
       cornerRadii=<0 0 0 0>
      }
      tintColor {colorUseRule=UseColor_Specified, specifiedColor=<rgba 0.0044 0.2384 1 1>}
      uVRegion {bIsValid=false, max=<0 0>, min=<0 0>}
     }
     undeterminedImage {
      bIsDynamicallyLoaded=false
      drawAs=Image
      imageType=Vector
      mirroring=NoMirror
      resourceName=../../../Engine/Content/Slate/Starship/CoreWidgets/CheckBox/indeterminate.svg
      resourceObject=null
      tiling=NoTile
      imageSize=<16 16>
      margin=<ltrb 0 0 0 0>
      outlineSettings {
       bUseBrushTransparency=false
       roundingType=HalfHeightRadius
       width=0
       color {colorUseRule=UseColor_Specified, specifiedColor=<rgba 0 0 0 0>}
       cornerRadii=<0 0 0 0>
      }
      tintColor {colorUseRule=UseColor_Specified, specifiedColor=<rgba 0 0.162 0.7454 1>}
      uVRegion {bIsValid=false, max=<0 0>, min=<0 0>}
     }
     undeterminedPressedImage {
      bIsDynamicallyLoaded=false
      drawAs=Image
      imageType=Vector
      mirroring=NoMirror
      resourceName=../../../Engine/Content/Slate/Starship/CoreWidgets/CheckBox/indeterminate.svg
      resourceObject=null
      tiling=NoTile
      imageSize=<16 16>
      margin=<ltrb 0 0 0 0>
      outlineSettings {
       bUseBrushTransparency=false
       roundingType=HalfHeightRadius
       width=0
       color {colorUseRule=UseColor_Specified, specifiedColor=<rgba 0 0 0 0>}
       cornerRadii=<0 0 0 0>
      }
      tintColor {colorUseRule=UseColor_Specified, specifiedColor=<rgba 0 0.0802 0.3515 1>}
      uVRegion {bIsValid=false, max=<0 0>, min=<0 0>}
     }
    }
   }
   slot {class=BorderSlot, properties {padding=<ltrb 3 3 3 3>}}
  }
  TextBlock_2 {
   widgetClassPath=&TextBlock
   parent=checkBox
   widget=TextBlock_2
   properties {
    text=No Collision LOD
    font {
     bForceMonospaced=false
     bMaterialIsStencil=false
     fontMaterial=null
     fontObject=&Roboto
     letterSpacing=0
     monospacedWidth=1
     size=9
     skewAmount=0
     typefaceFontName=Bold
     outlineSettings {
      bApplyOutlineToDropShadows=false
      bMiteredCorners=false
      bSeparateFillAlpha=false
      outlineMaterial=null
      outlineSize=0
      outlineColor=<rgba 0 0 0 1>
     }
    }
   }
   slot {class=PanelSlot}
  }
 ]
}
