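# Editor utility filter widget for the Asset Validation Tool that flags Texture2Ds whose largest effective dimension meets or exceeds a spin-box threshold.
# getTextureSizeAfterBias clamps each dimension to MaxTextureSize, then halves it per positive LODBias step.
# filterAsset returns an error struct coloured by size-to-threshold ratio, using zero severity when the short side is small.
# Base-class overrides store and restore the threshold and checkbox state through filter data.
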EditorUtilityWidgetBlueprint /Game/Blueprints/Blutilities/ForArtists/AssetValidationTool/SubWidgets/FilterWidgets/AVT_Tex_TooLarge_FilterWidget
graphs [
 EventGraph {nodes [
   n1 = "On Check State Changed (checkBox)"!() -> (OutputDelegate, then, bIsChecked:bool) {is=title K2Node_ComponentBoundEvent}
   > setFilterEnabled!(execute, self:AVT Base Filter Widget Object Reference, isEnabled:bool=@n1.bIsChecked) -> (then, changed:bool)
   n2 = "On Value Committed (spinTexSize)"!() -> (OutputDelegate, then, InValue:float, CommitMethod:ETextCommit Enum=Default) {is=title K2Node_ComponentBoundEvent}
   > n3 = Switch on ETextCommit!(execute, Selection:ETextCommit Enum=@n2.CommitMethod) -> (Default:Exec, OnEnter:Exec, OnUserMovedFocus:Exec, OnCleared:Exec) {is=title K2Node_SwitchEnum}
   triggerSaveGame(execute=[@n3.OnEnter, @n3.OnUserMovedFocus], self:AVT Base Filter Widget Object Reference)
  ]}
 filterAsset {nodes [
   n1 = filterAsset!() -> (then, asset:Object Reference) {is=title K2Node_FunctionEntry}
   > n2 = Cast To Texture2D(Object=@n1)
   > Set l_texture(l_texture:Texture 2D Object Reference=@n2) -> (Output_Get:Texture 2D Object Reference)
   > n3 = Sequence()
   > n4 = getTextureSizeAfterBias!(execute, self, texture:Texture 2D Object Reference=Get l_texture() -> (l_texture:Texture 2D Object Reference)) -> (then, x:int=0, y:int=0)
   > Set l_sizeX(l_sizeX:int=@n4) -> (Output_Get:int=0)
   > Set l_sizeY(l_sizeY:int=@n4.y) -> (Output_Get:int=0)
   > Set l_largestSize(l_largestSize:int="Max (Integer)"(A=@n4, B=@n4.y)) -> (Output_Get:int=0)
   > Set l_smallestSize(l_smallestSize:int="Min (Integer)"(A=@n4, B=@n4.y) -> (ReturnValue=0)) -> (Output_Get:int=0)
   n5 = Truncate(A=GetValue!(self:Spin Box Object Reference=Get spinTexSize() -> (spinTexSize:Spin Box Object Reference)) -> (ReturnValue:float))
   n6 = Branch(
    execute=@n3.then_1
    Condition="integer >= integer"!(A:int=Get l_largestSize() -> (l_largestSize:int=0), B:int=@n5) -> (ReturnValue:bool) {is=title K2Node_CallFunction}
   )
   n7 = Get l_largestSize() -> (l_largestSize=0)
   n8 = int / int!(A:int=Get l_largestSize() -> (l_largestSize=0), B:int=@n5) -> (ReturnValue:int=0) {is=title K2Node_CallFunction}
   n9 = Get l_smallestSize() -> (l_smallestSize:int=0)
   n10 = Select(
    Option 0:double=NormalizeToRange!(Value:double="To Float (Integer)"(InInt=@n8), RangeMin:double=1, RangeMax:double=2) -> (ReturnValue:double)
    Option 1:double=0
    Index=AND Boolean(A="integer < integer"!(A:int=@n9, B:int=257) -> (ReturnValue:bool) {is=title K2Node_CallFunction}, B="integer < integer"(A=@n9, B=@n5))
   ) -> (ReturnValue:double)
   n11 = "To String (Integer)"(InInt=Get l_largestSize() -> (l_largestSize=0))
   n12 = Get l_texture()
   n13 = "To String (Integer)"(InInt=Get MaxTextureSize!(self:Texture Object Reference=@n12) -> (MaxTextureSize:int=0))
   n14 = "To String (Integer)"(InInt=Get LODBias!(self:Texture Object Reference=@n12) -> (LODBias:int=0))
   n15 = "To String (Integer)"(InInt=Get l_sizeX() -> (l_sizeX:int=0))
   n16 = "To String (Integer)"(InInt=Get l_sizeY() -> (l_sizeY:int=0))
   > Return Node!(
    execute
    hasError:bool=true
    ErrorStruct:AVT Error Struct Structure=Make AVT Error Struct!(
     title_32_#1:String=Texture
     description_35_#2:String=Append(A="Large - ", B=@n11)
     tooltip_37_#3:String=Append!(
      A="Texture's largest dimension was "
      B=@n11
      C:String="\r\n"
      D:String=@n15
      E:String=x
      F:String=@n16
      G:String=" with Max Size: "
      H:String=@n13
      I:String=" and bias: "
      J:String=@n14
     ) -> (ReturnValue)
     color_40_#4:Linear Color Structure=lerpErrorColoring!(Alpha:double=@n10) -> (color:Linear Color Structure)
     csvLine_46_#5:String="To String (Integer)"(InInt=@n7)
     csvHeader_49_#6:String=Texture Size
     optionalAssetData_52_#7:Asset Data Structure=()
     optionalTypeTag_55_#8:String
     optionalOriginPackageName_62_#9:Name=None
     helpCommand_65_#10:String
    ) -> (AVT_Error_Struct:AVT Error Struct Structure) {is=title K2Node_MakeStruct}
   )
   Return Node(execute=@n6.else, hasError=false)
  ]}
 getTextureSizeAfterBias {nodes [
   n1 = getTextureSizeAfterBias!() -> (then, texture:Texture 2D Object Reference) {is=title K2Node_FunctionEntry}
   > Set l_texture(l_texture=@n1)
   > n2 = Sequence!(execute) -> (then_0, then_1, then_2:Exec)
   n3 = Get l_texture()
   n4 = Get l_sizes() -> (l_sizes:Array of Integers)
   > Add(TargetArray:Array of Integers=@n4, NewItem:"Integer (by ref)"=GetSizeX!(self:Texture 2D Object Reference=@n3) -> (ReturnValue:int=0)) {is=title K2Node_CallArrayFunction}
   > Add(TargetArray=@n4, NewItem=GetSizeY!(self:Texture 2D Object Reference=@n3) -> (ReturnValue:int=0))
   n5 = For Each Loop(Exec=@n2.then_1, Array:Array of Integers=Get l_sizes()) -> (Array Element:int)
   > Set l_largestSize(l_largestSize=@n5) -> (Output_Get=0)
   > n6 = Sequence!(execute) -> (then_0, then_1)
   n7 = Get MaxTextureSize(self=Get l_texture()) -> (MaxTextureSize=0)
   > n8 = Branch(Condition="integer > integer"(A=@n7, B=0) {is=title K2Node_CallFunction})
   > n9 = Set l_largestSize(l_largestSize="Min (Integer)"(A=@n5, B=@n7) -> (ReturnValue=0)) -> (Output_Get=0)
   n10 = Get LODBias(self=Get l_texture()) -> (LODBias=0)
   Branch(execute=[@n8.else, @n9.then], Condition="integer > integer"(A=@n10, B=0))
   > For Loop(LastIndex=int - int(A=@n10, B=1) {is=title K2Node_CallFunction})
   > Set l_largestSize(l_largestSize=int / int(A=Get l_largestSize() -> (l_largestSize=0), B=2) -> (ReturnValue=0)) -> (Output_Get=0)
   Set Array Elem!(
    execute=@n6.then_1
    TargetArray:Array of Integers=Get l_sizes()
    Index:int=@n5.Array Index
    Item:"Integer (by ref)"=Get l_largestSize() -> (l_largestSize=0)
    bSizeToFit:bool=false
   ) -> (then)
   n11 = Get l_sizes()
   Return Node!(execute=@n2.then_2, x:int="Get (a ref)"(Array=@n11, Dimension 1=0), y:int="Get (a ref)"(Array=@n11, Dimension 1=1))
  ]}
 validateFilterData {nodes [
   validateFilterData() {is=title K2Node_FunctionEntry}
   > Parent: validateFilterData!(execute) -> (then, " ":bool) {is=title K2Node_CallParentFunction}
   n1 = Get spinTexSize()
   > Set members in AVT Filter Data Struct!(
    execute:Exec
    StructRef:"AVT Filter Data Struct Structure (by ref)"=Get currentFilterData() -> (currentFilterData:AVT Filter Data Struct Structure)
    Floats_8_#11:"Array of Float (single-precision)s"=Make Array("[0]":float=GetValue(self=@n1)) -> (Array:"Array of Float (single-precision)s")
   ) -> (then:Exec, StructOut:"AVT Filter Data Struct Structure (by ref)") {is=title K2Node_SetFieldsInStruct}
   Return Node!(execute, " ":bool=false)
  ]}
 validateWidgets {nodes [
   validateWidgets() {is=title K2Node_FunctionEntry}
   > Branch(Condition="Not Equal (Name)"!(
     A:Name=Get!() -> (
      currentFilterData_Bools_2_#12:Array of Booleans
      currentFilterData_Strings_5_#13:Array of Strings
      currentFilterData_Floats_8_#11:"Array of Float (single-precision)s"
      currentFilterData_filterID_13_#14:Name
      currentFilterData_filterEnabled_15_#15:bool=~
      currentFilterData_hasHadDataSet_17_#16:bool=~
     ).currentFilterData_filterID_13_#14
     B:Name=None
    ) -> (ReturnValue:bool) {is=title K2Node_CallFunction})
   > Parent: validateWidgets!(execute) -> (then, " ":bool) {is=title K2Node_CallParentFunction}
   > n1 = Sequence()
   > SetIsChecked!(
    execute
    self:Check Box Object Reference=Get checkBox() -> (checkBox:Check Box Object Reference)
    InIsChecked:bool=getFilterEnabled!() -> (isEnabled:bool)
   ) -> (then)
   n2 = Break AVT Filter Data Struct!(AVT_FilterData_Struct:"AVT Filter Data Struct Structure (by ref)"=Get currentFilterData()) -> (
    Bools_2_#12:Array of Booleans
    Strings_5_#13:Array of Strings
    Floats_8_#11:"Array of Float (single-precision)s"
    filterID_13_#14:Name
    filterEnabled_15_#15:bool=~
    hasHadDataSet_17_#16:bool=~
   ) {is=title K2Node_BreakStruct}
   Branch(
    execute=@n1.then_1
    Condition=Is Valid Index!(TargetArray:"Array of Float (single-precision)s"=@n2.Floats_8_#11, IndexToTest:int) -> (ReturnValue:bool)
   )
   > SetValue!(
    execute
    self:Spin Box Object Reference=Get spinTexSize()
    NewValue:float="Get (a copy)"(Array:"Array of Float (single-precision)s"=@n2.Floats_8_#11) -> (Output:float)
   ) -> (then)
   Return Node(" "=false)
  ]}
]
variables []
widgets {
 info {
  parentClass=&/AVT_Base_FilterWidget.AVT_Base_FilterWidget_C
  rootWidgetClass=&CanvasPanel
  widgetCount=6
 }
 widgets [
  CanvasPanel_0 {widgetClassPath=&CanvasPanel, widget=CanvasPanel_0}
  Border_4 {
   widgetClassPath=&Border
   parent=CanvasPanel_0
   widget=Border_4
   properties {brushColor=<rgba 0.0296 0.0296 0.0296 1>, padding=<ltrb 3 3 3 3>}
   slot {class=CanvasPanelSlot, properties {layoutData {alignment=<0 0>, anchors {maximum=<0 0>, minimum=<0 0>}, offsets=<ltrb 0 0 161.3036 30>}}}
  }
  HorizontalBox_0 {
   widgetClassPath=&HorizontalBox
   parent=Border_4
   widget=HorizontalBox_0
   slot {class=BorderSlot, properties {padding=<ltrb 3 3 3 3>}}
  }
  checkBox {
   widgetClassPath=/Script/UMG.CheckBox
   bIsVariable=true
   parent=HorizontalBox_0
   widget=checkBox
   properties {
    checkedState=Checked
    toolTipText=Check if meshes have any LODs at all.
    widgetStyle {
     checkBoxType=CheckBox
     backgroundHoveredImage {
      bIsDynamicallyLoaded=false
      drawAs=RoundedBox
      imageType=NoImage
      mirroring=NoMirror
      resourceName=None
      resourceObject=null
      tiling=NoTile
      imageSize=<18 18>
      margin=<ltrb 0 0 0 0>
      outlineSettings {
       bUseBrushTransparency=false
       roundingType=FixedRadius
       width=1
       color {colorUseRule=UseColor_ColorTable, specifiedColor=<rgba 0 0 0 0>}
       cornerRadii=<3 3 3 3>
      }
      tintColor {colorUseRule=UseColor_Specified, specifiedColor=<rgba 0.0048 0.0048 0.0048 1>}
      uVRegion {bIsValid=false, max=<0 0>, min=<0 0>}
     }
     backgroundImage {
      bIsDynamicallyLoaded=false
      drawAs=RoundedBox
      imageType=NoImage
      mirroring=NoMirror
      resourceName=None
      resourceObject=null
      tiling=NoTile
      imageSize=<18 18>
      margin=<ltrb 0 0 0 0>
      outlineSettings {
       bUseBrushTransparency=false
       roundingType=FixedRadius
       width=1
       color {colorUseRule=UseColor_ColorTable, specifiedColor=<rgba 0 0 0 0>}
       cornerRadii=<3 3 3 3>
      }
      tintColor {colorUseRule=UseColor_Specified, specifiedColor=<rgba 0.0048 0.0048 0.0048 1>}
      uVRegion {bIsValid=false, max=<0 0>, min=<0 0>}
     }
     backgroundPressedImage {
      bIsDynamicallyLoaded=false
      drawAs=RoundedBox
      imageType=NoImage
      mirroring=NoMirror
      resourceName=None
      resourceObject=null
      tiling=NoTile
      imageSize=<18 18>
      margin=<ltrb 0 0 0 0>
      outlineSettings {
       bUseBrushTransparency=false
       roundingType=FixedRadius
       width=1
       color {colorUseRule=UseColor_ColorTable, specifiedColor=<rgba 0 0 0 0>}
       cornerRadii=<3 3 3 3>
      }
      tintColor {colorUseRule=UseColor_Specified, specifiedColor=<rgba 0.0048 0.0048 0.0048 1>}
      uVRegion {bIsValid=false, max=<0 0>, min=<0 0>}
     }
     borderBackgroundColor {colorUseRule=UseColor_Specified, specifiedColor=<rgba 1 1 1 1>}
     checkedForeground {colorUseRule=UseColor_Specified, specifiedColor=<rgba 1 1 1 1>}
     checkedHoveredForeground {colorUseRule=UseColor_Specified, specifiedColor=<rgba 1 1 1 1>}
     checkedHoveredImage {
      bIsDynamicallyLoaded=false
      drawAs=Image
      imageType=Vector
      mirroring=NoMirror
      resourceName=../../../Engine/Content/Slate/Starship/CoreWidgets/CheckBox/check.svg
      resourceObject=null
      tiling=NoTile
      imageSize=<16 16>
      margin=<ltrb 0 0 0 0>
      outlineSettings {
       bUseBrushTransparency=false
       roundingType=HalfHeightRadius
       width=0
       color {colorUseRule=UseColor_Specified, specifiedColor=<rgba 0 0 0 0>}
       cornerRadii=<0 0 0 0>
      }
      tintColor {colorUseRule=UseColor_Specified, specifiedColor=<rgba 0.0044 0.2384 1 1>}
      uVRegion {bIsValid=false, max=<0 0>, min=<0 0>}
     }
     checkedImage {
      bIsDynamicallyLoaded=false
      drawAs=Image
      imageType=Vector
      mirroring=NoMirror
      resourceName=../../../Engine/Content/Slate/Starship/CoreWidgets/CheckBox/check.svg
      resourceObject=null
      tiling=NoTile
      imageSize=<16 16>
      margin=<ltrb 0 0 0 0>
      outlineSettings {
       bUseBrushTransparency=false
       roundingType=HalfHeightRadius
       width=0
       color {colorUseRule=UseColor_Specified, specifiedColor=<rgba 0 0 0 0>}
       cornerRadii=<0 0 0 0>
      }
      tintColor {colorUseRule=UseColor_Specified, specifiedColor=<rgba 0 0.162 0.7454 1>}
      uVRegion {bIsValid=false, max=<0 0>, min=<0 0>}
     }
     checkedPressedForeground {colorUseRule=UseColor_Specified, specifiedColor=<rgba 1 1 1 1>}
     checkedPressedImage {
      bIsDynamicallyLoaded=false
      drawAs=Image
      imageType=Vector
      mirroring=NoMirror
      resourceName=../../../Engine/Content/Slate/Starship/CoreWidgets/CheckBox/check.svg
      resourceObject=null
      tiling=NoTile
      imageSize=<16 16>
      margin=<ltrb 0 0 0 0>
      outlineSettings {
       bUseBrushTransparency=false
       roundingType=HalfHeightRadius
       width=0
       color {colorUseRule=UseColor_Specified, specifiedColor=<rgba 0 0 0 0>}
       cornerRadii=<0 0 0 0>
      }
      tintColor {colorUseRule=UseColor_Specified, specifiedColor=<rgba 0 0.0802 0.3515 1>}
      uVRegion {bIsValid=false, max=<0 0>, min=<0 0>}
     }
     checkedSlateSound {resourceObject=null}
     foregroundColor {colorUseRule=UseColor_Specified, specifiedColor=<rgba 1 1 1 1>}
     hoveredForeground {colorUseRule=UseColor_Specified, specifiedColor=<rgba 1 1 1 1>}
     hoveredSlateSound {resourceObject=null}
     padding=<ltrb 2 0 0 0>
     pressedForeground {colorUseRule=UseColor_Specified, specifiedColor=<rgba 1 1 1 1>}
     uncheckedHoveredImage {
      bIsDynamicallyLoaded=false
      drawAs=NoDrawType
      imageType=NoImage
      mirroring=NoMirror
      resourceName=None
      resourceObject=null
      tiling=NoTile
      imageSize=<0 0>
      margin=<ltrb 0 0 0 0>
      outlineSettings {
       bUseBrushTransparency=false
       roundingType=HalfHeightRadius
       width=0
       color {colorUseRule=UseColor_Specified, specifiedColor=<rgba 0 0 0 0>}
       cornerRadii=<0 0 0 0>
      }
      tintColor {colorUseRule=UseColor_Specified, specifiedColor=<rgba 1 1 1 1>}
      uVRegion {bIsValid=false, max=<0 0>, min=<0 0>}
     }
     uncheckedImage {
      bIsDynamicallyLoaded=false
      drawAs=NoDrawType
      imageType=NoImage
      mirroring=NoMirror
      resourceName=None
      resourceObject=null
      tiling=NoTile
      imageSize=<0 0>
      margin=<ltrb 0 0 0 0>
      outlineSettings {
       bUseBrushTransparency=false
       roundingType=HalfHeightRadius
       width=0
       color {colorUseRule=UseColor_Specified, specifiedColor=<rgba 0 0 0 0>}
       cornerRadii=<0 0 0 0>
      }
      tintColor {colorUseRule=UseColor_Specified, specifiedColor=<rgba 1 1 1 1>}
      uVRegion {bIsValid=false, max=<0 0>, min=<0 0>}
     }
     uncheckedPressedImage {
      bIsDynamicallyLoaded=false
      drawAs=NoDrawType
      imageType=NoImage
      mirroring=NoMirror
      resourceName=None
      resourceObject=null
      tiling=NoTile
      imageSize=<0 0>
      margin=<ltrb 0 0 0 0>
      outlineSettings {
       bUseBrushTransparency=false
       roundingType=HalfHeightRadius
       width=0
       color {colorUseRule=UseColor_Specified, specifiedColor=<rgba 0 0 0 0>}
       cornerRadii=<0 0 0 0>
      }
      tintColor {colorUseRule=UseColor_Specified, specifiedColor=<rgba 1 1 1 1>}
      uVRegion {bIsValid=false, max=<0 0>, min=<0 0>}
     }
     uncheckedSlateSound {resourceObject=null}
     undeterminedForeground {colorUseRule=UseColor_Specified, specifiedColor=<rgba 1 1 1 1>}
     undeterminedHoveredImage {
      bIsDynamicallyLoaded=false
      drawAs=Image
      imageType=Vector
      mirroring=NoMirror
      resourceName=../../../Engine/Content/Slate/Starship/CoreWidgets/CheckBox/indeterminate.svg
      resourceObject=null
      tiling=NoTile
      imageSize=<16 16>
      margin=<ltrb 0 0 0 0>
      outlineSettings {
       bUseBrushTransparency=false
       roundingType=HalfHeightRadius
       width=0
       color {colorUseRule=UseColor_Specified, specifiedColor=<rgba 0 0 0 0>}
       cornerRadii=<0 0 0 0>
      }
      tintColor {colorUseRule=UseColor_Specified, specifiedColor=<rgba 0.0044 0.2384 1 1>}
      uVRegion {bIsValid=false, max=<0 0>, min=<0 0>}
     }
     undeterminedImage {
      bIsDynamicallyLoaded=false
      drawAs=Image
      imageType=Vector
      mirroring=NoMirror
      resourceName=../../../Engine/Content/Slate/Starship/CoreWidgets/CheckBox/indeterminate.svg
      resourceObject=null
      tiling=NoTile
      imageSize=<16 16>
      margin=<ltrb 0 0 0 0>
      outlineSettings {
       bUseBrushTransparency=false
       roundingType=HalfHeightRadius
       width=0
       color {colorUseRule=UseColor_Specified, specifiedColor=<rgba 0 0 0 0>}
       cornerRadii=<0 0 0 0>
      }
      tintColor {colorUseRule=UseColor_Specified, specifiedColor=<rgba 0 0.162 0.7454 1>}
      uVRegion {bIsValid=false, max=<0 0>, min=<0 0>}
     }
     undeterminedPressedImage {
      bIsDynamicallyLoaded=false
      drawAs=Image
      imageType=Vector
      mirroring=NoMirror
      resourceName=../../../Engine/Content/Slate/Starship/CoreWidgets/CheckBox/indeterminate.svg
      resourceObject=null
      tiling=NoTile
      imageSize=<16 16>
      margin=<ltrb 0 0 0 0>
      outlineSettings {
       bUseBrushTransparency=false
       roundingType=HalfHeightRadius
       width=0
       color {colorUseRule=UseColor_Specified, specifiedColor=<rgba 0 0 0 0>}
       cornerRadii=<0 0 0 0>
      }
      tintColor {colorUseRule=UseColor_Specified, specifiedColor=<rgba 0 0.0802 0.3515 1>}
      uVRegion {bIsValid=false, max=<0 0>, min=<0 0>}
     }
    }
   }
   slot {class=HorizontalBoxSlot, properties {verticalAlignment=VAlign_Center}}
  }
  TextBlock_2 {
   widgetClassPath=&TextBlock
   parent=HorizontalBox_0
   widget=TextBlock_2
   properties {
    text="Size >= "
    font {
     bForceMonospaced=false
     bMaterialIsStencil=false
     fontMaterial=null
     fontObject=&Roboto
     letterSpacing=0
     monospacedWidth=1
     size=14
     skewAmount=0
     typefaceFontName=Bold
     outlineSettings {
      bApplyOutlineToDropShadows=false
      bMiteredCorners=false
      bSeparateFillAlpha=false
      outlineMaterial=null
      outlineSize=0
      outlineColor=<rgba 0 0 0 1>
     }
    }
   }
   slot {class=HorizontalBoxSlot, properties {verticalAlignment=VAlign_Center}}
  }
  spinTexSize {
   widgetClassPath=/Script/UMG.SpinBox
   bIsVariable=true
   parent=HorizontalBox_0
   widget=spinTexSize
   properties {
    bOverride_MaxSliderValue=true
    bOverride_MaxValue=true
    bOverride_MinSliderValue=true
    bOverride_MinValue=true
    delta=1
    maxSliderValue=8192
    maxValue=8192
    value=2048
    foregroundColor {colorUseRule=UseColor_Specified, specifiedColor=<rgba 1 1 1 1>}
    widgetStyle {
     activeBackgroundBrush {
      bIsDynamicallyLoaded=false
      drawAs=RoundedBox
      imageType=NoImage
      mirroring=NoMirror
      resourceName=None
      resourceObject=null
      tiling=NoTile
      imageSize=<0 0>
      margin=<ltrb 0 0 0 0>
      outlineSettings {
       bUseBrushTransparency=false
       roundingType=FixedRadius
       width=1
       color {colorUseRule=UseColor_ColorTable, specifiedColor=<rgba 0 0 0 0>}
       cornerRadii=<4 4 4 4>
      }
      tintColor {colorUseRule=UseColor_Specified, specifiedColor=<rgba 0.0048 0.0048 0.0048 1>}
      uVRegion {bIsValid=false, max=<0 0>, min=<0 0>}
     }
     activeFillBrush {
      bIsDynamicallyLoaded=false
      drawAs=RoundedBox
      imageType=NoImage
      mirroring=NoMirror
      resourceName=None
      resourceObject=null
      tiling=NoTile
      imageSize=<0 0>
      margin=<ltrb 0 0 0 0>
      outlineSettings {
       bUseBrushTransparency=false
       roundingType=FixedRadius
       width=1
       color {colorUseRule=UseColor_Specified, specifiedColor=<rgba 0 0 0 0>}
       cornerRadii=<4 4 4 4>
      }
      tintColor {colorUseRule=UseColor_Specified, specifiedColor=<rgba 0 0.162 0.7454 1>}
      uVRegion {bIsValid=false, max=<0 0>, min=<0 0>}
     }
     arrowsImage {
      bIsDynamicallyLoaded=false
      drawAs=NoDrawType
      imageType=NoImage
      mirroring=NoMirror
      resourceName=None
      resourceObject=null
      tiling=NoTile
      imageSize=<0 0>
      margin=<ltrb 0 0 0 0>
      outlineSettings {
       bUseBrushTransparency=false
       roundingType=HalfHeightRadius
       width=0
       color {colorUseRule=UseColor_Specified, specifiedColor=<rgba 0 0 0 0>}
       cornerRadii=<0 0 0 0>
      }
      tintColor {colorUseRule=UseColor_Specified, specifiedColor=<rgba 1 1 1 1>}
      uVRegion {bIsValid=false, max=<0 0>, min=<0 0>}
     }
     backgroundBrush {
      bIsDynamicallyLoaded=false
      drawAs=RoundedBox
      imageType=NoImage
      mirroring=NoMirror
      resourceName=None
      resourceObject=null
      tiling=NoTile
      imageSize=<0 0>
      margin=<ltrb 0 0 0 0>
      outlineSettings {
       bUseBrushTransparency=false
       roundingType=FixedRadius
       width=1
       color {colorUseRule=UseColor_ColorTable, specifiedColor=<rgba 0 0 0 0>}
       cornerRadii=<4 4 4 4>
      }
      tintColor {colorUseRule=UseColor_Specified, specifiedColor=<rgba 0.0048 0.0048 0.0048 1>}
      uVRegion {bIsValid=false, max=<0 0>, min=<0 0>}
     }
     foregroundColor {colorUseRule=UseColor_Specified, specifiedColor=<rgba 1 1 1 1>}
     hoveredBackgroundBrush {
      bIsDynamicallyLoaded=false
      drawAs=RoundedBox
      imageType=NoImage
      mirroring=NoMirror
      resourceName=None
      resourceObject=null
      tiling=NoTile
      imageSize=<0 0>
      margin=<ltrb 0 0 0 0>
      outlineSettings {
       bUseBrushTransparency=false
       roundingType=FixedRadius
       width=1
       color {colorUseRule=UseColor_ColorTable, specifiedColor=<rgba 0 0 0 0>}
       cornerRadii=<4 4 4 4>
      }
      tintColor {colorUseRule=UseColor_Specified, specifiedColor=<rgba 0.0048 0.0048 0.0048 1>}
      uVRegion {bIsValid=false, max=<0 0>, min=<0 0>}
     }
     hoveredFillBrush {
      bIsDynamicallyLoaded=false
      drawAs=RoundedBox
      imageType=NoImage
      mirroring=NoMirror
      resourceName=None
      resourceObject=null
      tiling=NoTile
      imageSize=<0 0>
      margin=<ltrb 0 0 0 0>
      outlineSettings {
       bUseBrushTransparency=false
       roundingType=FixedRadius
       width=1
       color {colorUseRule=UseColor_Specified, specifiedColor=<rgba 0 0 0 0>}
       cornerRadii=<4 4 4 4>
      }
      tintColor {colorUseRule=UseColor_Specified, specifiedColor=<rgba 0.0953 0.0953 0.0953 1>}
      uVRegion {bIsValid=false, max=<0 0>, min=<0 0>}
     }
     inactiveFillBrush {
      bIsDynamicallyLoaded=false
      drawAs=RoundedBox
      imageType=NoImage
      mirroring=NoMirror
      resourceName=None
      resourceObject=null
      tiling=NoTile
      imageSize=<0 0>
      margin=<ltrb 0 0 0 0>
      outlineSettings {
       bUseBrushTransparency=false
       roundingType=FixedRadius
       width=1
       color {colorUseRule=UseColor_Specified, specifiedColor=<rgba 0 0 0 0>}
       cornerRadii=<4 4 4 4>
      }
      tintColor {colorUseRule=UseColor_Specified, specifiedColor=<rgba 0.0395 0.0395 0.0395 1>}
      uVRegion {bIsValid=false, max=<0 0>, min=<0 0>}
     }
     insetPadding=<ltrb 1 1 1 1>
     textPadding=<ltrb 7 2.5 7 2.5>
    }
   }
   slot {class=HorizontalBoxSlot, properties {verticalAlignment=VAlign_Center, size {sizeRule=Fill, value=1}}}
  }
 ]
}
