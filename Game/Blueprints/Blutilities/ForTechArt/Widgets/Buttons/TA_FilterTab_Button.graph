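# Editor tab button for tech-art tools, pairing a selectable icon and label with a category filter checkbox.
# Left-click toggles selection, which heightens the tab and clears its dimming cover.
# Filter state loads from and saves to the list item data's bools, unbinding the checkbox delegate during programmatic changes.
# Preview mouse-down records Shift/Ctrl for the checked-status event; Ctrl-click rebroadcasts the current state and selects only this tab in its parent list.
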WidgetBlueprint /Game/Blueprints/Blutilities/ForTechArt/Widgets/Buttons/TA_FilterTab_Button
graphs [
 EventGraph {nodes [
   n1 = Event handleClicked!() -> (OutputDelegate, then, InputEvent:"Pointer Event Structure (by ref)") {is=title K2Node_Event}
   > Is Valid(InputObject=Get currentListItemData() -> (currentListItemData:Object Reference))
   > Parent: handleClicked!(execute, InputEvent:"Pointer Event Structure (by ref)"=@n1.InputEvent) -> (then) {is=title K2Node_CallParentFunction}
   > Branch(Condition="Equal (Key)"!(
     A:Key Structure=Get Effecting Button!(Input:"Pointer Event Structure (by ref)"=@n1.InputEvent) -> (ReturnValue:Key Structure) {is=title K2Node_CallFunction}
     B:Key Structure=LeftMouseButton
    ) -> (ReturnValue:bool))
   > setSelected!(
    execute
    self:TA Base Editor Button Widget Object Reference
    selected:bool=NOT Boolean(A=getSelected!() -> (isSelected:bool))
    throwEvent:bool=true
   ) -> (then)
   n2 = Event handleNewListDataSet!() -> (OutputDelegate, then, currentListItemData:Object Reference) {is=title K2Node_Event}
   > Parent: handleNewListDataSet!(execute, currentListItemData:Object Reference=@n2.currentListItemData) -> (then) {is=title K2Node_CallParentFunction}
   > n3 = Cast To TA_GenericListItemData!(execute, Object=@n2.currentListItemData) -> (then, CastFailed, AsTA Generic List Item Data:TA Generic List Item Data Object Reference) {is=title K2Node_DynamicCast}
   n4 = Get bools!(self:TA Generic List Item Data Object Reference=@n3) -> (bools:Array of Booleans)
   > n5 = Branch(Condition=Is Valid Index!(TargetArray:Array of Booleans=@n4, IndexToTest:int) -> (ReturnValue:bool))
   > setFilterEnabled!(execute, self, enabled:bool="Get (a copy)"(Array:Array of Booleans=@n4) -> (Output:bool)) -> (then)
   setFilterEnabled(execute=@n5.else, enabled=true)
   n6 = Event setSelected!() -> (OutputDelegate, then, selected:bool, throwEvent:bool) {is=title K2Node_Event}
   > Parent: setSelected!(execute, selected:bool=@n6.selected, throwEvent:bool=true) -> (then) {is=title K2Node_CallParentFunction}
   > SetColorAndOpacity(
    self:Image Object Reference=Get imgCover() -> (imgCover:Image Object Reference)
    InColorAndOpacity:Linear Color Structure=Select(
     Option 0:Linear Color Structure=Get initialCoverColor() -> (initialCoverColor:Linear Color Structure)
     Option 1:Linear Color Structure=(R=0,G=0,B=0,A=0)
     Index=@n6.selected
    ) -> (ReturnValue:Linear Color Structure)
   )
   > SetHeightOverride(
    self=Get SizeBox_0() -> (SizeBox_0:Size Box Object Reference)
    InHeightOverride=Select(Option 0:double=28, Option 1:double=31, Index=@n6.selected) -> (ReturnValue:double)
   )
   Event Construct()
   > Parent: Construct()
   > Set initialCoverColor(initialCoverColor:Linear Color Structure=Get ColorAndOpacity!(self:Image Object Reference=Get imgCover()) -> (ColorAndOpacity:Linear Color Structure)) -> (Output_Get:Linear Color Structure)
   > Bind Event to On Check State Changed(self:Check Box Object Reference=Get cbMain() -> (cbMain:Check Box Object Reference), Delegate=Create Event()) {is=title K2Node_AddDelegate}
  ]}
 OnMouseButtonUp {nodes [
   n1 = OnMouseButtonUp!() -> (then, MyGeometry:Geometry Structure, MouseEvent:"Pointer Event Structure (by ref)") {is=title K2Node_FunctionEntry}
   > Set isShiftDown(isShiftDown:bool=false) -> (Output_Get:bool)
   > Set isControlDown(isControlDown:bool=false) -> (Output_Get:bool)
   > n2 = Parent: OnMouseButtonUp!(execute, MyGeometry:Geometry Structure=@n1, MouseEvent:"Pointer Event Structure (by ref)"=@n1.MouseEvent) -> (then, ReturnValue:Event Reply Structure) {is=title K2Node_CallParentFunction}
   > Return Node!(execute, ReturnValue:Event Reply Structure=@n2)
  ]}
 OnPreviewMouseButtonDown {nodes [
   n1 = OnPreviewMouseButtonDown!() -> (then, MyGeometry:Geometry Structure, MouseEvent:"Pointer Event Structure (by ref)") {is=title K2Node_FunctionEntry}
   n2 = GetInputEventFromPointerEvent!(Event:"Pointer Event Structure (by ref)"=@n1.MouseEvent) -> (ReturnValue:Input Event Structure)
   n3 = Is Shift Down!(Input:"Input Event Structure (by ref)"=@n2) -> (ReturnValue:bool)
   > Set isShiftDown(isShiftDown=@n3)
   n4 = Is Control Down!(Input:"Input Event Structure (by ref)"=@n2) -> (ReturnValue:bool)
   > Set isControlDown(isControlDown=@n4)
   > n5 = Branch(Condition=@n4)
   > handleCheckBoxChanged!(execute, self, isChecked:bool=IsChecked(self=Get cbMain())) -> (then)
   > DetectDragIfPressed!(execute, PointerEvent:"Pointer Event Structure (by ref)"=@n1.MouseEvent, DragKey:Key Structure=Get Effecting Button(Input=@n1.MouseEvent)) -> (then, ReturnValue:Event Reply Structure)
   > n6 = FindParentWidgetOfType!(
    execute
    StartingWidget:Widget Object Reference=Self-Reference()
    Type:Widget Class Reference=/Game/Blueprints/Blutilities/ForTechArt/Widgets/List/TA_Generic01_List.TA_Generic01_List_C
   ) -> (then, ReturnValue:TA Generic 01 List Object Reference)
   > n7 = Cast To TA_Generic01_List!(execute, Object=@n6) -> (then, CastFailed, AsTA Generic 01 List:TA Generic 01 List Object Reference) {is=title K2Node_DynamicCast}
   > setSelectedItems!(
    execute
    self:TA Generic 01 List Object Reference=@n7
    items:Array of User Widget Object References=Make Array("[0]":User Widget Object Reference=Self-Reference()) -> (Array:Array of User Widget Object References)
   ) -> (then)
   > Return Node(ReturnValue=Handled())
   n8 = Parent: OnPreviewMouseButtonDown!(execute=@n5.else, MyGeometry:Geometry Structure=@n1, MouseEvent:"Pointer Event Structure (by ref)"=@n1.MouseEvent) -> (then, ReturnValue:Event Reply Structure) {is=title K2Node_CallParentFunction}
   > Return Node(ReturnValue=@n8)
   OR Boolean(A=@n3, B=@n4)
  ]}
 getBackgroundWidget {nodes [
   getBackgroundWidget() {is=title K2Node_FunctionEntry}
   > Return Node!(execute, backgroundWidget:Widget Object Reference=Get borderMain() -> (borderMain:Border Object Reference))
  ]}
 getFilterEnabled {nodes [
   getFilterEnabled() {is=title K2Node_FunctionEntry}
   > Return Node!(execute, enabled:bool=IsChecked(self=Get cbMain()))
  ]}
 getLabelWidget {nodes [
   getLabelWidget() {is=title K2Node_FunctionEntry}
   > Return Node!(execute, labelWidget:Widget Object Reference=Get txtMain() -> (txtMain:Text Object Reference))
  ]}
 handleCheckBoxChanged {nodes [
   n1 = handleCheckBoxChanged!() -> (then, isChecked:bool=~) {is=title K2Node_FunctionEntry}
   > Call On Checked Status Changed!(
    execute:Exec
    self:TA Filter Tab Button Object Reference
    widget:TA Filter Tab Button Object Reference=Self-Reference()
    data:Object Reference=Get currentListItemData()
    isChecked:bool=@n1
    isShiftDown:bool=Get isShiftDown() -> (isShiftDown:bool)
    isControlDown:bool=Get isControlDown() -> (isControlDown:bool)
   ) -> (then:Exec) {is=title K2Node_CallDelegate}
  ]}
 onCheckedStatusChanged {nodes [onCheckedStatusChanged!() -> (then, widget:TA Filter Tab Button Object Reference, data:Object Reference, isChecked:bool=~, isShiftDown:bool=~, isControlDown:bool=~) {is=title K2Node_FunctionEntry}]}
 setFilterEnabled {nodes [
   n1 = setFilterEnabled!() -> (then, enabled:bool=~) {is=title K2Node_FunctionEntry}
   > n2 = Sequence!(execute) -> (then_0, then_1, then_2:Exec)
   > Unbind all Events from On Check State Changed!(execute:Exec, self:Check Box Object Reference=Get cbMain()) -> (then:Exec) {is=title K2Node_ClearDelegate}
   SetIsChecked!(execute=@n2.then_1, self:Check Box Object Reference=Get cbMain(), InIsChecked:bool=@n1) -> (then)
   > n3 = Cast To TA_GenericListItemData(Object=Get currentListItemData())
   > Set bools!(
    execute
    bools:Array of Booleans=Make Array("[0]":bool=@n1) -> (Array:Array of Booleans)
    self:TA Generic List Item Data Object Reference=@n3
   ) -> (then, Output_Get:Array of Booleans)
   Bind Event to On Check State Changed(execute=@n2.then_2, self=Get cbMain(), Delegate=Create Event())
  ]}
 validateVisuals {nodes [
   validateVisuals() {is=title K2Node_FunctionEntry}
   > Parent: validateVisuals!(execute) -> (then, " ":bool) {is=title K2Node_CallParentFunction}
   > n1 = Sequence!(execute) -> (then_0, then_1)
   > n2 = Is Valid(InputObject=Break Slate Brush!(SlateBrush:"Slate Brush Structure (by ref)"=Get iconBrush() -> (iconBrush:Slate Brush Structure)) -> (ResourceObject:Object Reference, OutlineSettings:Slate Brush Outline Settings Structure) {is=title K2Node_BreakStruct})
   SetVisibility(execute=@n2.Is Not Valid, self=Get imageIcon() -> (imageIcon:Image Object Reference), InVisibility=Collapsed)
   SetBrush(execute=@n2.Is Valid, self=Get imageIcon(), InBrush=Get iconBrush())
   > SetVisibility(self=Get imageIcon(), InVisibility=HitTestInvisible)
   > SetRenderTransform!(
    execute
    self:Widget Object Reference=Get imageIcon()
    InTransform:Widget Transform Structure=Get iconTransform() -> (iconTransform:Widget Transform Structure)
   ) -> (then)
   n3 = Get label() -> (label:String)
   SetVisibility(
    execute=@n1.then_1
    self=Get txtMain()
    InVisibility=Select(Option 0:ESlateVisibility Enum=Collapsed, Option 1:ESlateVisibility Enum=HitTestInvisible, Index="Not Equal Exactly (String)"(A=@n3)) -> (ReturnValue:ESlateVisibility Enum)
   )
   Return Node!(execute, " ":bool=false)
  ]}
]
variables [
 iconBrush {
  type=SlateBrush
  default {
   bIsDynamicallyLoaded=false
   drawAs=Image
   imageType=NoImage
   mirroring=NoMirror
   resourceName=None
   resourceObject=null
   tiling=NoTile
   imageSize=<32 32>
   margin=<ltrb 0 0 0 0>
   outlineSettings {
    bUseBrushTransparency=false
    roundingType=HalfHeightRadius
    width=0
    color {colorUseRule=UseColor_Specified, specifiedColor=<rgba 0 0 0 0>}
    cornerRadii=<0 0 0 1>
   }
   tintColor {colorUseRule=UseColor_Specified, specifiedColor=<rgba 1 1 1 1>}
   uVRegion {bIsValid=false, max=<0 0>, min=<0 0>}
  }
 }
 iconTransform {type=WidgetTransform, default {angle=0, scale=<1 1>, shear=<0 0>, translation=<0 0>}}
 initialCoverColor {type=LinearColor, default=<rgba 0 0 0 0>}
 isControlDown {type=boolean}
 isShiftDown {type=boolean}
 onCheckedStatusChanged {}
]
widgets {
 info {
  inheritedWidgetCount=0
  parentClass=&/TA_BaseEditorButtonWidget.TA_BaseEditorButtonWidget_C
  rootWidgetClass=&CanvasPanel
  widgetCount=11
 }
 widgets [
  CanvasPanel_1 {widgetClassPath=&CanvasPanel, widget=CanvasPanel_1}
  SizeBox_1 {
   widgetClassPath=&SizeBox
   parent=CanvasPanel_1
   widget=SizeBox_1
   properties {bOverride_HeightOverride=true, heightOverride=31}
   slot {
    class=CanvasPanelSlot
    properties {bAutoSize=true, layoutData {alignment=<0 0>, anchors {maximum=<0 0>, minimum=<0 0>}, offsets=<ltrb 0 0 300 30>}}
   }
  }
  Overlay_0 {widgetClassPath=&Overlay, parent=SizeBox_1, widget=Overlay_0, slot {class=SizeBoxSlot}}
  SizeBox_0 {
   widgetClassPath=&SizeBox
   bIsVariable=true
   parent=Overlay_0
   widget=SizeBox_0
   properties {bOverride_HeightOverride=true, heightOverride=30}
   slot {class=OverlaySlot, properties {horizontalAlignment=HAlign_Fill, verticalAlignment=VAlign_Bottom}}
  }
  borderMain {
   widgetClassPath=&Border
   bIsVariable=true
   parent=SizeBox_0
   widget=borderMain
   properties {
    verticalAlignment=VAlign_Center
    background {
     bIsDynamicallyLoaded=false
     drawAs=Box
     imageType=NoImage
     mirroring=NoMirror
     resourceName=None
     resourceObject=/Game/Blueprints/Blutilities/ForTechArt/ArtAssets/Textures/T_TA_TabBG_01_Up
     tiling=NoTile
     imageSize=<32 32>
     margin=<ltrb 0.25 0.25 0.25 0.25>
     outlineSettings {
      bUseBrushTransparency=false
      roundingType=HalfHeightRadius
      width=0
      color {colorUseRule=UseColor_Specified, specifiedColor=<rgba 0 0 0 0>}
      cornerRadii=<0 0 0 0>
     }
     tintColor {colorUseRule=UseColor_Specified, specifiedColor=<rgba 0.7 0.7 0.7 1>}
     uVRegion {bIsValid=false, max=<0 0>, min=<0 0>}
    }
    padding=<ltrb 8 5 8 4>
   }
   slot {class=SizeBoxSlot}
  }
  HorizontalBox_0 {
   widgetClassPath=&HorizontalBox
   parent=borderMain
   widget=HorizontalBox_0
   slot {class=BorderSlot, properties {verticalAlignment=VAlign_Center, padding=<ltrb 8 5 8 4>}}
  }
  cbMain {
   widgetClassPath=/Script/UMG.CheckBox
   bIsVariable=true
   parent=HorizontalBox_0
   widget=cbMain
   properties {
    toolTipText=Include this category in filters
    widgetStyle {
     checkBoxType=CheckBox
     backgroundHoveredImage {
      bIsDynamicallyLoaded=false
      drawAs=NoDrawType
      imageType=NoImage
      mirroring=NoMirror
      resourceName=None
      resourceObject=null
      tiling=NoTile
      imageSize=<0 0>
      margin=<ltrb 0 0 0 0>
      outlineSettings {
       bUseBrushTransparency=false
       roundingType=HalfHeightRadius
       width=0
       color {colorUseRule=UseColor_Specified, specifiedColor=<rgba 0 0 0 0>}
       cornerRadii=<0 0 0 0>
      }
      tintColor {colorUseRule=UseColor_Specified, specifiedColor=<rgba 1 1 1 1>}
      uVRegion {bIsValid=false, max=<0 0>, min=<0 0>}
     }
     backgroundImage {
      bIsDynamicallyLoaded=false
      drawAs=NoDrawType
      imageType=NoImage
      mirroring=NoMirror
      resourceName=None
      resourceObject=null
      tiling=NoTile
      imageSize=<0 0>
      margin=<ltrb 0 0 0 0>
      outlineSettings {
       bUseBrushTransparency=false
       roundingType=HalfHeightRadius
       width=0
       color {colorUseRule=UseColor_Specified, specifiedColor=<rgba 0 0 0 0>}
       cornerRadii=<0 0 0 0>
      }
      tintColor {colorUseRule=UseColor_Specified, specifiedColor=<rgba 1 1 1 1>}
      uVRegion {bIsValid=false, max=<0 0>, min=<0 0>}
     }
     backgroundPressedImage {
      bIsDynamicallyLoaded=false
      drawAs=NoDrawType
      imageType=NoImage
      mirroring=NoMirror
      resourceName=None
      resourceObject=null
      tiling=NoTile
      imageSize=<0 0>
      margin=<ltrb 0 0 0 0>
      outlineSettings {
       bUseBrushTransparency=false
       roundingType=HalfHeightRadius
       width=0
       color {colorUseRule=UseColor_Specified, specifiedColor=<rgba 0 0 0 0>}
       cornerRadii=<0 0 0 0>
      }
      tintColor {colorUseRule=UseColor_Specified, specifiedColor=<rgba 1 1 1 1>}
      uVRegion {bIsValid=false, max=<0 0>, min=<0 0>}
     }
     borderBackgroundColor {colorUseRule=UseColor_Specified, specifiedColor=<rgba 1 1 1 1>}
     checkedForeground {colorUseRule=UseColor_Foreground, specifiedColor=<rgba 1 0 1 1>}
     checkedHoveredForeground {colorUseRule=UseColor_Foreground, specifiedColor=<rgba 1 0 1 1>}
     checkedHoveredImage {
      bIsDynamicallyLoaded=false
      drawAs=Image
      imageType=FullColor
      mirroring=NoMirror
      resourceName=../../../Engine/Content/Slate/Common/CheckBox_Checked_Hovered.png
      resourceObject=/Game/Blueprints/Blutilities/ForTechArt/ArtAssets/Textures/Checkbox/TA_Checkbox_Checked
      tiling=NoTile
      imageSize=<20 20>
      margin=<ltrb 0 0 0 0>
      outlineSettings {
       bUseBrushTransparency=false
       roundingType=HalfHeightRadius
       width=0
       color {colorUseRule=UseColor_Specified, specifiedColor=<rgba 0 0 0 0>}
       cornerRadii=<0 0 0 0>
      }
      tintColor {colorUseRule=UseColor_Specified, specifiedColor=<rgba 1 1 1 1>}
      uVRegion {bIsValid=false, max=<0 0>, min=<0 0>}
     }
     checkedImage {
      bIsDynamicallyLoaded=false
      drawAs=Image
      imageType=FullColor
      mirroring=NoMirror
      resourceName=../../../Engine/Content/Slate/Common/CheckBox_Checked_Hovered.png
      resourceObject=&TA_Checkbox_Checked
      tiling=NoTile
      imageSize=<20 20>
      margin=<ltrb 0 0 0 0>
      outlineSettings {
       bUseBrushTransparency=false
       roundingType=HalfHeightRadius
       width=0
       color {colorUseRule=UseColor_Specified, specifiedColor=<rgba 0 0 0 0>}
       cornerRadii=<0 0 0 0>
      }
      tintColor {colorUseRule=UseColor_Specified, specifiedColor=<rgba 0.7135 0.7135 0.7135 1>}
      uVRegion {bIsValid=false, max=<0 0>, min=<0 0>}
     }
     checkedPressedForeground {colorUseRule=UseColor_Foreground, specifiedColor=<rgba 1 0 1 1>}
     checkedPressedImage {
      bIsDynamicallyLoaded=false
      drawAs=Image
      imageType=FullColor
      mirroring=NoMirror
      resourceName=../../../Engine/Content/Slate/Common/CheckBox_Checked.png
      resourceObject=&TA_Checkbox_Checked
      tiling=NoTile
      imageSize=<20 20>
      margin=<ltrb 0 0 0 0>
      outlineSettings {
       bUseBrushTransparency=false
       roundingType=HalfHeightRadius
       width=0
       color {colorUseRule=UseColor_Specified, specifiedColor=<rgba 0 0 0 0>}
       cornerRadii=<0 0 0 0>
      }
      tintColor {colorUseRule=UseColor_Specified, specifiedColor=<rgba 1 1 1 1>}
      uVRegion {bIsValid=false, max=<0 0>, min=<0 0>}
     }
     checkedSlateSound {resourceObject=null}
     foregroundColor {colorUseRule=UseColor_Foreground, specifiedColor=<rgba 1 0 1 1>}
     hoveredForeground {colorUseRule=UseColor_Foreground, specifiedColor=<rgba 1 0 1 1>}
     hoveredSlateSound {resourceObject=null}
     padding=<ltrb 2 0 0 0>
     pressedForeground {colorUseRule=UseColor_Foreground, specifiedColor=<rgba 1 0 1 1>}
     uncheckedHoveredImage {
      bIsDynamicallyLoaded=false
      drawAs=Image
      imageType=FullColor
      mirroring=NoMirror
      resourceName=../../../Engine/Content/Slate/Common/CheckBox.png
      resourceObject=/Game/Blueprints/Blutilities/ForTechArt/ArtAssets/Textures/Checkbox/TA_Checkbox_Unchecked
      tiling=NoTile
      imageSize=<20 20>
      margin=<ltrb 0 0 0 0>
      outlineSettings {
       bUseBrushTransparency=false
       roundingType=HalfHeightRadius
       width=0
       color {colorUseRule=UseColor_Specified, specifiedColor=<rgba 0 0 0 0>}
       cornerRadii=<0 0 0 0>
      }
      tintColor {colorUseRule=UseColor_Specified, specifiedColor=<rgba 1 1 1 1>}
      uVRegion {bIsValid=false, max=<0 0>, min=<0 0>}
     }
     uncheckedImage {
      bIsDynamicallyLoaded=false
      drawAs=Image
      imageType=FullColor
      mirroring=NoMirror
      resourceName=../../../Engine/Content/Slate/Common/CheckBox.png
      resourceObject=&TA_Checkbox_Unchecked
      tiling=NoTile
      imageSize=<20 20>
      margin=<ltrb 0 0 0 0>
      outlineSettings {
       bUseBrushTransparency=false
       roundingType=HalfHeightRadius
       width=0
       color {colorUseRule=UseColor_Specified, specifiedColor=<rgba 0 0 0 0>}
       cornerRadii=<0 0 0 0>
      }
      tintColor {colorUseRule=UseColor_Specified, specifiedColor=<rgba 0.6302 0.6302 0.6302 1>}
      uVRegion {bIsValid=false, max=<0 0>, min=<0 0>}
     }
     uncheckedPressedImage {
      bIsDynamicallyLoaded=false
      drawAs=Image
      imageType=FullColor
      mirroring=NoMirror
      resourceName=../../../Engine/Content/Slate/Common/CheckBox_Hovered.png
      resourceObject=&TA_Checkbox_Unchecked
      tiling=NoTile
      imageSize=<20 20>
      margin=<ltrb 0 0 0 0>
      outlineSettings {
       bUseBrushTransparency=false
       roundingType=HalfHeightRadius
       width=0
       color {colorUseRule=UseColor_Specified, specifiedColor=<rgba 0 0 0 0>}
       cornerRadii=<0 0 0 0>
      }
      tintColor {colorUseRule=UseColor_Specified, specifiedColor=<rgba 0.5 0.5 0.5 1>}
      uVRegion {bIsValid=false, max=<0 0>, min=<0 0>}
     }
     uncheckedSlateSound {resourceObject=null}
     undeterminedForeground {colorUseRule=UseColor_Foreground, specifiedColor=<rgba 1 0 1 1>}
     undeterminedHoveredImage {
      bIsDynamicallyLoaded=false
      drawAs=Image
      imageType=FullColor
      mirroring=NoMirror
      resourceName=../../../Engine/Content/Slate/Common/CheckBox_Undetermined_Hovered.png
      resourceObject=null
      tiling=NoTile
      imageSize=<20 20>
      margin=<ltrb 0 0 0 0>
      outlineSettings {
       bUseBrushTransparency=false
       roundingType=HalfHeightRadius
       width=0
       color {colorUseRule=UseColor_Specified, specifiedColor=<rgba 0 0 0 0>}
       cornerRadii=<0 0 0 0>
      }
      tintColor {colorUseRule=UseColor_Specified, specifiedColor=<rgba 1 1 1 1>}
      uVRegion {bIsValid=false, max=<0 0>, min=<0 0>}
     }
     undeterminedImage {
      bIsDynamicallyLoaded=false
      drawAs=Image
      imageType=FullColor
      mirroring=NoMirror
      resourceName=../../../Engine/Content/Slate/Common/CheckBox_Undetermined.png
      resourceObject=null
      tiling=NoTile
      imageSize=<20 20>
      margin=<ltrb 0 0 0 0>
      outlineSettings {
       bUseBrushTransparency=false
       roundingType=HalfHeightRadius
       width=0
       color {colorUseRule=UseColor_Specified, specifiedColor=<rgba 0 0 0 0>}
       cornerRadii=<0 0 0 0>
      }
      tintColor {colorUseRule=UseColor_Specified, specifiedColor=<rgba 1 1 1 1>}
      uVRegion {bIsValid=false, max=<0 0>, min=<0 0>}
     }
     undeterminedPressedImage {
      bIsDynamicallyLoaded=false
      drawAs=Image
      imageType=FullColor
      mirroring=NoMirror
      resourceName=../../../Engine/Content/Slate/Common/CheckBox_Undetermined_Hovered.png
      resourceObject=null
      tiling=NoTile
      imageSize=<20 20>
      margin=<ltrb 0 0 0 0>
      outlineSettings {
       bUseBrushTransparency=false
       roundingType=HalfHeightRadius
       width=0
       color {colorUseRule=UseColor_Specified, specifiedColor=<rgba 0 0 0 0>}
       cornerRadii=<0 0 0 0>
      }
      tintColor {colorUseRule=UseColor_Specified, specifiedColor=<rgba 0.5 0.5 0.5 1>}
      uVRegion {bIsValid=false, max=<0 0>, min=<0 0>}
     }
    }
   }
   slot {class=HorizontalBoxSlot}
  }
  imageIcon {
   widgetClassPath=&Image
   bIsVariable=true
   parent=HorizontalBox_0
   widget=imageIcon
   slot {class=HorizontalBoxSlot, properties {horizontalAlignment=HAlign_Center, verticalAlignment=VAlign_Center}}
  }
  ScaleBox_0 {
   widgetClassPath=&ScaleBox
   parent=HorizontalBox_0
   widget=ScaleBox_0
   properties {clipping=ClipToBounds, stretchDirection=DownOnly, renderTransform {angle=0, scale=<1 1>, shear=<0 0>, translation=<0 1>}}
   slot {class=HorizontalBoxSlot, properties {verticalAlignment=VAlign_Center, size {sizeRule=Fill, value=1}}}
  }
  txtMain {
   widgetClassPath=&TextBlock
   bIsVariable=true
   parent=ScaleBox_0
   widget=txtMain
   properties {
    clipping=ClipToBounds
    justification=Center
    colorAndOpacity {colorUseRule=UseColor_Specified, specifiedColor=<rgba 0.0104 0.0104 0.0104 1>}
    font {
     bForceMonospaced=false
     bMaterialIsStencil=false
     fontMaterial=null
     fontObject=&Roboto
     letterSpacing=0
     monospacedWidth=1
     size=12
     skewAmount=0
     typefaceFontName=Bold
     outlineSettings {
      bApplyOutlineToDropShadows=true
      bMiteredCorners=false
      bSeparateFillAlpha=false
      outlineMaterial=null
      outlineSize=1
      outlineColor=<rgba 0 0 0 0.071>
     }
    }
    shadowColorAndOpacity=<rgba 0 0 0 0.524>
    shadowOffset=<1.2 1.2>
   }
   slot {class=ScaleBoxSlot}
  }
  imgCover {
   widgetClassPath=&Image
   bIsVariable=true
   parent=Overlay_0
   widget=imgCover
   properties {
    brush {
     bIsDynamicallyLoaded=false
     drawAs=Image
     imageType=NoImage
     mirroring=NoMirror
     resourceName=None
     resourceObject=null
     tiling=NoTile
     imageSize=<32 4>
     margin=<ltrb 0 0 0 0>
     outlineSettings {
      bUseBrushTransparency=false
      roundingType=HalfHeightRadius
      width=0
      color {colorUseRule=UseColor_Specified, specifiedColor=<rgba 0 0 0 0>}
      cornerRadii=<0 0 0 0>
     }
     tintColor {colorUseRule=UseColor_Specified, specifiedColor=<rgba 1 1 1 1>}
     uVRegion {bIsValid=false, max=<0 0>, min=<0 0>}
    }
    colorAndOpacity=<rgba 0 0 0 0.2>
   }
   slot {class=OverlaySlot, properties {horizontalAlignment=HAlign_Fill, verticalAlignment=VAlign_Bottom}}
  }
 ]
}
